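# Idle cosmetic Niagara system for the Voltaic Huntress character, with burst-spawned, non-expiring particles pinned to the skeletal mesh.
# Capsule_Glow (mesh) and pulsing Glow_Portal (sprite) take color and alpha from arrays indexed by User.Variant, dimmed by time of day.
# Smoke_Ribbons sets VisibilityTag from Variant across thirteen mesh renderers, with variant-specific night dimming.
# The ribbons also store smoothed world position and feed a wrapped integrator value into a dynamic material parameter.
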NiagaraSystem /BRCosmetics/Characters/Player/Female/Medium/Bodies/F_MED_VoltaicHuntress/FX/NS_VoltaicHuntress_Idle
dependencies {
 usedDataInterfaces [
  &NiagaraDataInterfaceArrayFloat
  &NiagaraDataInterfaceArrayColor
  &NiagaraDataInterfaceSkeletalMesh
 ]
 usedDynamicInputs [&SetBoolByFloatComparison, &VectorLength]
 usedModules [
  &self:SystemSpawnScript.SystemScriptSource.SystemScriptGraph.NiagaraNodeAssignment_11.SetVariables_#1
  &FN_Integrator
  &SystemState
 ]
}
emitters {
 Capsule_Glow {
  emitterUpdateScript [
   &EmitterState(
    Life Cycle Mode=::NewEnumerator1
    Loop Behavior=::NewEnumerator1
    Loop Duration Mode=::NewEnumerator1
    Scalability Mode=::NewEnumerator1
    UseLoopDelay=false
   )
   &SpawnBurst_Instantaneous(Spawn Count=10, Use Spawn Probability=false, UseLoopCountLimit=false)
  ]
  particleSpawnScript [&InitializeParticle(
    Color Mode=::NewEnumerator0
    Lifetime=5
    Mesh Scale=&Multiply_VectorByFloat(Vector=<0.6 0.6 0.5>, Float=0.2)
    Mesh Scale Mode=::NewEnumerator4
    UsePositionOffset=false
    Write Material Random=true
   )]
  particleUpdateScript [
   &ParticleState(Kill Particles When Lifetime Has Elapsed=false)
   &SkeletalMeshLocation(
    Bone / Socket Index=1
    Bone Sampling Mode=::NewEnumerator4
    MaskSpawnGroup=false
    Sampled Position Offset=<-3.5 -4.5 0>
    Skeletal Mesh=$System.Skeletal Mesh
   )
   &CameraOffset(Camera Offset Amount=1)
   &ScaleColor(
    Scale Alpha=&Multiply_Float(
     A=&Multiply_Float(
      A=0.05
      B=&SelectFloatFromArray(
       Direct Array Index=&MakeIntFromFloat(Float=$User.Variant)
       Float Selection Array=DataInterface() {propertyValues {FloatData [1,0.4,0.4,1.5,2,1.5]}}
      )
     )
     B=&FNTimeOfDayScalar(Night Scale=0.7, Morning Scale=0.8, Evening Scale=0.8)
    )
    Scale RGB=&MakeVectorFromLinearColorRGB(Color=&SelectLinearColorFromArray(
      Array Sampling Mode=::NewEnumerator1
      Direct Array Index=&MakeIntFromFloat(Float=$User.Variant)
      Color Selection Array=DataInterface() {propertyValues {ColorData [<rgba 0 0.2333 1 1>, <rgba 0.3875 0.9301 0 1>, <rgba 1 0.65 0 1>, <rgba 0.7575 0.15 1.5 1>, <rgba 0 0.9646 1.5 1>, <rgba 0 0.5833 1 1>]}}
      Random Seed=?
     ))
    ScaleA=true
    ScaleRGB=true
   )
  ]
  rendererClasses [&NiagaraMeshRendererProperties]
  renderers [{rendererClass=&NiagaraMeshRendererProperties, rendererIndex=0}]
 }
 Glow_Portal {
  emitterUpdateScript [
   &EmitterState(
    Inactive Response=::NewEnumerator1
    Life Cycle Mode=::NewEnumerator1
    Loop Behavior=::NewEnumerator1
    Loop Duration Mode=::NewEnumerator1
    UseLoopDelay=false
   )
   &SpawnBurst_Instantaneous(Use Spawn Probability=false, UseLoopCountLimit=false)
  ]
  particleSpawnScript [
   &InitializeParticle(
    Lifetime=5
    Position Offset=<0 0.1 0>
    Sprite Size Mode=::NewEnumerator3
    Uniform Sprite Size=8
    UsePositionOffset=false
    Write Material Random=true
   )
   &SystemLocation(MaskSpawnGroup=false, Offset=<0 -6 0>)
  ]
  particleUpdateScript [
   &ParticleState(Kill Particles When Lifetime Has Elapsed=false)
   &SkeletalMeshLocation(Bone Sampling Mode=::NewEnumerator4, MaskSpawnGroup=false, Sampled Position Offset=<-2.5 -0.25 4.5>, Skeletal Mesh=$System.Skeletal Mesh)
   &ScaleColor(
    Scale RGB=&VectorFromFloat(Value=&Waveform(
      Waveform Count=::NewEnumerator1
      "Waveform [1]"=::NewEnumerator0
      "[1] Frequency"=0.618
      "[1] Amplitude Min/Max"=<0.25 0.5>
      "[2] Average"=false
      "[2] Input Position"=0
      "[2] Phase"=$Emitter.Age
      "[2] Phase Scale"=1
      "[2] Compound Frequency"=2
      "[2] Frequency"=1
      "[2] Amplitude Min/Max"=<0.25 0.5>
      "[2] Scale"=1
      "[2] Bias"=0
      "[2] Exponent"=1
      "[2] Pulse Duration"=0.1
      "[2] Clip Triangle Min/Max"=<0 0>
      "[2] Random Seed"=0
      "Waveform [3] Blend"=::NewEnumerator5
     ))
    ScaleA=false
    ScaleRGB=false
   )
   &CameraOffset(Camera Offset Amount=2)
   &ScaleColor(
    Scale Alpha=&Multiply_Float(
     A=&Multiply_Float(
      A=0.5
      B=&SelectFloatFromArray(
       Direct Array Index=&MakeIntFromFloat(Float=$User.Variant)
       Float Selection Array=DataInterface() {propertyValues {FloatData [0.5,0.5,0.5,0.5,0.5,0.5]}}
      )
     )
     B=&FNTimeOfDayScalar(Night Scale=0.7, Morning Scale=0.8, Evening Scale=0.8)
    )
    Scale RGB=&MakeVectorFromLinearColorRGB(Color=&SelectLinearColorFromArray(
      Array Sampling Mode=::NewEnumerator1
      Direct Array Index=&MakeIntFromFloat(Float=$User.Variant)
      Color Selection Array=DataInterface() {propertyValues {ColorData [
         <rgba 0 0.2333 1 1>
         <rgba 0.3875 0.9301 0 0.3>
         <rgba 1 0.65 0 0.25>
         <rgba 0.4174 0.1491 0.7454 0.1>
         <rgba 0 0.5394 0.8388 0.5>
         <rgba 0 0.4707 0.807 0.15>
        ]}}
      Random Seed=?
     ))
    ScaleA=true
    ScaleRGB=true
   ) {module=ScaleColor002}
  ]
  rendererClasses [&NiagaraSpriteRendererProperties]
  renderers [{rendererClass=&NiagaraSpriteRendererProperties}]
 }
 Smoke_Ribbons {
  emitterUpdateScript [
   &EmitterState(
    Life Cycle Mode=::NewEnumerator1
    Loop Behavior=::NewEnumerator1
    Loop Duration Mode=::NewEnumerator1
    Scalability Mode=::NewEnumerator1
    UseLoopDelay=false
   )
   &SpawnBurst_Instantaneous(Use Spawn Probability=false, UseLoopCountLimit=false)
  ]
  particleSpawnScript [&InitializeParticle(Color Mode=::NewEnumerator0, Lifetime=5, Mesh Scale Mode=::NewEnumerator3, UsePositionOffset=false, Write Material Random=true)]
  particleUpdateScript [
   &ParticleState(Kill Particles When Lifetime Has Elapsed=false)
   /SetVariables(Particles.VisibilityTag=&MakeIntFromFloat(Float=$User.Variant)) {isSetParameters=true}
   &SkeletalMeshLocation(
    Bone Sampling Mode=::NewEnumerator4
    MaskSpawnGroup=false
    Orientation Sampling=::NewEnumerator1
    Sampled Position Offset=<-2.5 -0.25 4.5>
    Skeletal Mesh=$System.Skeletal Mesh
   )
   &InterpolateOverTime(
    Rate Of Change=6
    Target Vector Value:Vector3f=&ConvertPositionToVector(
     Input Position=&TransformPosition(Position=$Particles.Position, Source Space=::Local, Destination Space=::World)
     Convert To Absolute Large World Vector=false
    )
    Variable Type=::NewEnumerator2
   )
   /SetVariables(Particles.Interpolate Dynamic=&MakeFloat4FromVectorAndFloat("Vector (RGB)"=$StackContext.InterpolateOverTime.InterpolateOverTime_V3_.Moving Average, "Float (Alpha)"=0.15)) {isSetParameters=true}
   &DynamicMaterialParameters(
    Index 0 Param 4=&Frac_Float(Float=$System.FN_Integrator002.Value)
    Module.Index0Param0Write=false
    Module.Index0Param1Write=false
    Module.Index0Param2Write=false
    Module.Index0Param3Write=true
   )
   &ScaleColor(
    Scale Alpha=&FNTimeOfDayScalar(
     Night Scale=&SelectFloatFromArray(
      Direct Array Index=&MakeIntFromFloat(Float=$User.Variant)
      Float Selection Array=DataInterface() {propertyValues {FloatData [0.66,0.66,0.4,0.4,0.3,0.66]}}
     )
     Morning Scale=0.77
     Evening Scale=0.7
    )
    ScaleA=true
    ScaleRGB=true
   )
  ]
  rendererClasses [&NiagaraMeshRendererProperties]
  renderers [
   {rendererClass=&NiagaraMeshRendererProperties, rendererIndex=0}
   {rendererClass=&NiagaraMeshRendererProperties, rendererIndex=1}
   {rendererClass=&NiagaraMeshRendererProperties, rendererIndex=2}
   {rendererClass=&NiagaraMeshRendererProperties, rendererIndex=3}
   {rendererClass=&NiagaraMeshRendererProperties, rendererIndex=4}
   {rendererClass=&NiagaraMeshRendererProperties, rendererIndex=5}
   {rendererClass=&NiagaraMeshRendererProperties, rendererIndex=6}
   {rendererClass=&NiagaraMeshRendererProperties, rendererIndex=7}
   {rendererClass=&NiagaraMeshRendererProperties, rendererIndex=8}
   {rendererClass=&NiagaraMeshRendererProperties, rendererIndex=9}
   {rendererClass=&NiagaraMeshRendererProperties, rendererIndex=10}
   {rendererClass=&NiagaraMeshRendererProperties, rendererIndex=11}
   {rendererClass=&NiagaraMeshRendererProperties, rendererIndex=12}
  ]
 }
}
summary {
 systemName=NS_VoltaicHuntress_Idle
 emitters [
  {emitterName=Smoke_Ribbons, rendererClasses [&NiagaraMeshRendererProperties]}
  {emitterName=Glow_Portal, rendererClasses [&NiagaraSpriteRendererProperties]}
  {emitterName=Capsule_Glow, rendererClasses [&NiagaraMeshRendererProperties]}
 ]
 userVariables [
  User.bIsFrontEnd {type=NiagaraBool}
  User.bIsFrontEndPreview {type=NiagaraBool}
  User.Variant {type=NiagaraFloat, defaultValue=0, description=""}
 ]
}
