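# Material function for the Voltaic Huntress accessory that draws a glowing circle effect with fake depth into material attributes.
# Bump-offset parallax UVs, blended by a radial fade, feed two calls to MF_VoltaicHuntress_Accessory_SDF, the second time-offset.
# A static switch picks one or two-color lines, composited over a two-color background with a flickering, parallaxed center falloff.
# The result drives emissive over black base color, with parameterized roughness and specular.
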MaterialFunction /BRCosmetics/Characters/Player/Female/Medium/Bodies/F_MED_VoltaicHuntress/Materials/MaterialFunctions/MF_VoltaicHuntress_Accessory
graph {nodes [
  n1 = VectorParameter() {group=VHAccessory, parameterName=LinesColor2, defaultValue=<rgba 3 0 0.1236 1>}
  n2 = VectorParameter() {group=VHAccessory, parameterName=LinesColor, defaultValue=<rgba 0 1.2932 3 1>}
  n3 = MaterialFunctionCall!(DepthUV=NamedRerouteUsage(), CircleCenterPosition=NamedRerouteUsage(), UVs=NamedRerouteUsage(), TimeOffset) {materialFunction=&/MF_VoltaicHuntress_Accessory_SDF}
  n4 = MaterialFunctionCall(DepthUV=NamedRerouteUsage(), CircleCenterPosition=NamedRerouteUsage(), UVs=NamedRerouteUsage(), TimeOffset=Constant() {r=5}) {materialFunction=&MF_VoltaicHuntress_Accessory_SDF}
  n5 = Saturate(None=Add(A=Multiply(A=@n4.Result, B=@n1.A), B=@n3.Result))
  n6 = LinearInterpolate(A=NamedRerouteUsage(), B=Multiply(A=LinearInterpolate(A=@n1.RGB, B=@n2.RGB, Alpha=@n3.Result), B=@n5), Alpha=@n5)
  n7 = LinearInterpolate(A=NamedRerouteUsage(), B=@n2.RGB, Alpha=@n3.Result)
  n8 = SetMaterialAttributes!(
   MaterialAttributes=FunctionInput() {inputType=FunctionInput_MaterialAttributes}
   EmissiveColor=NamedRerouteDeclaration(None=StaticSwitchParameter(True=@n6, False=@n7) {group=VHAccessory, parameterName=UseSecondLines}) {name=CircleEffect, nodeColor=<rgba 1 0 0.0703 1>}
   Roughness=ScalarParameter() {defaultValue=0.15, group=VHAccessory, parameterName=Roughness}
   BaseColor=Constant()
   Specular=ScalarParameter() {defaultValue=0.5, group=VHAccessory, parameterName=Specular}
  ) {attributeSetTypes [
    #1
    #2
    #3
    #4
   ]}
  FunctionOutput(None=@n8) {outputName=Emissive}
  n9 = NamedRerouteUsage()
  n10 = ScalarParameter() {defaultValue=0.6, group=VHAccessory, parameterName=CircleEffectDepth}
  n11 = ScalarParameter() {defaultValue=1.2, group=VHAccessory, parameterName=CircleBGDepth}
  n12 = Length(None=BumpOffset(Coordinate=Add(A=@n9) {constB=-0.5}, Height=Multiply(A=@n10, B=@n11) {constB=1.5}) {heightRatio=-1}) {is=class MaterialExpressionLength}
  n13 = Power(
   Base=SmoothStep(Value=OneMinus(None=@n12)) {constMin=0.8}
   Exp=ScalarParameter() {defaultValue=6, group=VHAccessory, parameterName=CircleBGSmoothness}
  )
  n14 = Constant() {r=0.5}
  n15 = Multiply(
   A=Multiply(A=MaterialFunctionCall!(Time Mult=@n14) {materialFunction=/Game/Effects/Fort_Effects/MaterialFunctions/MF_Flicker}.Raw Flicker) {constB=0.075}
   B=ScalarParameter() {defaultValue=1, parameterName=CircleBGFlickerIntensity}
  )
  n16 = Multiply(
   A=ScalarParameter() {defaultValue=3, group=VHAccessory, parameterName=CircleBGIntensity}
   B=NamedRerouteDeclaration(None=Add(A=@n15)) {name=Flicker, nodeColor=<rgba 0.6406 0 1 1>}
  )
  NamedRerouteDeclaration(None=Saturate(None=Multiply(A=@n13, B=@n16))) {name=Circle Background, nodeColor=<rgba 0 0.8594 1 1>}
  n17 = Subtract(
   A=TextureCoordinate()
   B=FunctionInput(Preview=Constant2Vector()) {bUsePreviewValueAsDefault=true, inputName=UVPositionOffset, inputType=FunctionInput_Vector2}
  )
  n18 = BumpOffset(
   Coordinate=NamedRerouteDeclaration(None=@n17) {name=UV, nodeColor=<rgba 0 1 0.6484 1>}
   Height=ScalarParameter() {defaultValue=0.6, group=VHAccessory, parameterName=CircleEffectDepth}
   HeightRatioInput=ScalarParameter() {defaultValue=-1, group=VHAccessory, parameterName=CircleEffectDepthRatio}
  ) {heightRatio=-1}
  n19 = NamedRerouteUsage()
  NamedRerouteDeclaration(None=LinearInterpolate(A=@n17, B=NamedRerouteDeclaration(None=@n18) {name=CircleCenterPosition, nodeColor=<rgba 0.1016 0 1 1>}, Alpha=OneMinus(None=@n19))) {name=DepthUV, nodeColor=<rgba 0.1016 1 0 1>}
  NamedRerouteDeclaration(None=LinearInterpolate(
    A=VectorParameter() {group=VHAccessory, parameterName=BG_InnerColor, defaultValue=<rgba 0.026 0 0.3958 1>}.RGB
    B=VectorParameter() {group=VHAccessory, parameterName=BG_OutterColor, defaultValue=<rgba 0.3231 0.2395 1 1>}.RGB
    Alpha=NamedRerouteUsage()
   )) {name=Background, nodeColor=<rgba 0 1 0.7188 1>}
  n20 = Multiply(A=Subtract(A=NamedRerouteUsage(), B=Constant() {r=0.5})) {constB=10}
  n21 = Multiply(A=@n20, B=@n20)
  n22 = ComponentMask(None=@n21) {r=true}
  n23 = ComponentMask(None=@n21) {g=true}
  NamedRerouteDeclaration(None=SquareRoot(None=Add(A=@n22, B=@n23))) {name=BackFadeMask, nodeColor=<rgba 0 1 0.4375 1>}
  NamedRerouteUsage()
 ]}
