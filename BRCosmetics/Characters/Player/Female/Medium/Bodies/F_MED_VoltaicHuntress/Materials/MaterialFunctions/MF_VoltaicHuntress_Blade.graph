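# Material function for the Voltaic Huntress blade, overriding incoming material attributes with black base color, parameterized specular and roughness, and animated emissive.
# A vertical UV gradient, peaking off-center, blends two blade colors beneath a detail color.
# The detail mask combines two offset alpha samples of a rotated, panning mask texture, distorted by its slower-scrolling red channel and sharpened with power and smoothstep.
# The output is named Emissive but carries the full attribute set.
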MaterialFunction /BRCosmetics/Characters/Player/Female/Medium/Bodies/F_MED_VoltaicHuntress/Materials/MaterialFunctions/MF_VoltaicHuntress_Blade
graph {nodes [
  n1 = VectorParameter() {group=VHBlade, parameterName=BladeColor1, defaultValue=<rgba 0.0172 0.1189 0.1823 1>}
  n2 = VectorParameter() {group=VHBlade, parameterName=BladeColor2, defaultValue=<rgba 0.1286 0.6386 2 1>}
  n3 = NamedRerouteUsage()
  n4 = ScalarParameter() {defaultValue=0.85, group=VHBlade, parameterName=BladeDetailColorOpacity}
  n5 = Multiply(A=TextureCoordinate(), B=Constant2Vector() {g=1, r=2.5})
  n6 = Multiply(A=TextureCoordinate()) {constB=0.75}
  n7 = Constant2Vector() {g=0.5, r=0.75}
  n8 = TextureSample(UVs=Panner(Coordinate=Multiply(A=@n6, B=@n7)) {speedY=0.05}) {
   samplerType=SAMPLERTYPE_LinearColor
   texture=/BRCosmetics/Characters/Player/Female/Medium/Bodies/F_MED_VoltaicHuntress/Materials/T_VoltaicHuntress_Mask
  }
  n9 = Add(
   A=MaterialFunctionCall(UVs=Panner(Coordinate=@n5) {speedY=-0.2}, "Rotation Angle (0-1)"=Constant() {r=0.25}) {materialFunction=&CustomRotator}.Rotated Values
   B=Multiply(A=@n8.R) {constB=0.3}
  )
  n10 = TextureSample(UVs=Add(A=@n9) {constB=0.2}) {
   samplerType=SAMPLERTYPE_LinearColor
   texture=&T_VoltaicHuntress_Mask
  }
  n11 = ScalarParameter() {defaultValue=2, group=VHBlade}
  n12 = TextureSample(UVs=@n9) {
   samplerType=SAMPLERTYPE_LinearColor
   texture=&T_VoltaicHuntress_Mask
  }
  n13 = LinearInterpolate(
   A=SmoothStep(Min=ScalarParameter() {defaultValue=0.1, group=VHBlade, parameterName=Param_2}, Value=Power(Base=@n10.A, Exp=@n11)) {constMin=0.35}
   Alpha=SmoothStep(Min=ScalarParameter() {defaultValue=0.01, group=VHBlade, parameterName=Param_1}, Value=Power(Base=@n12.A, Exp=@n11)) {constMax=0.2, constMin=0.35}
  ) {constAlpha=0.2}
  n14 = SetMaterialAttributes!(
   MaterialAttributes=FunctionInput() {inputType=FunctionInput_MaterialAttributes}
   EmissiveColor=LinearInterpolate(
    A=LinearInterpolate(A=@n1.RGB, B=@n2.RGB, Alpha=@n3)
    B=VectorParameter() {group=VHBlade, parameterName=BladeDetailColor, defaultValue=<rgba 0.1194 0 1 1>}.RGB
    Alpha=Multiply(A=@n4, B=@n13)
   )
   BaseColor=Constant()
   Specular=ScalarParameter() {defaultValue=0.5, group=VHBlade, parameterName=Specular}
   Roughness=ScalarParameter() {defaultValue=0.15, group=VHBlade, parameterName=Roughness}
  ) {attributeSetTypes [
    #1
    #2
    #3
    #4
   ]}
  FunctionOutput(None=@n14) {outputName=Emissive}
  n15 = ComponentMask(None=TextureCoordinate()) {g=true}
  n16 = OneMinus(None=Abs(None=Subtract(A=@n15) {constB=0.6}))
  NamedRerouteDeclaration(None=Saturate(None=Power(Base=@n16) {constExponent=3})) {name=TopdownGradient, nodeColor=<rgba 0.0781 1 0 1>}
  NamedRerouteUsage()
 ]}
