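# Material function for the Voltaic Huntress character body that draws a glowing, flickering capsule with electrical static inside.
# A UV position input centers radial masks, with BumpOffset faking cylinder depth, while summed sine waves drive flicker.
# Two MF_VoltaicHuntress_Static layers at different depths and time offsets form the static, with an optional secondary color.
# It blends static over the two-color background as emissive in the incoming material attributes, also setting base color and roughness.
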MaterialFunction /BRCosmetics/Characters/Player/Female/Medium/Bodies/F_MED_VoltaicHuntress/Materials/MaterialFunctions/MF_VoltaicHuntress_Capsule
graph {nodes [
  n1 = Multiply(
   A=Power(Base=NamedRerouteUsage(), Exp=ScalarParameter() {defaultValue=1.5, group=VHCapsule, parameterName=BGPower}) {constExponent=1}
   B=ScalarParameter() {defaultValue=0.5, group=VHCapsule, parameterName=BGIntensity}
  ) {constB=100}
  n2 = ScalarParameter() {defaultValue=1.75, parameterName=FlickerIntensity}
  n3 = Multiply(A=NamedRerouteUsage(), B=@n2) {constB=1.5}
  n4 = LinearInterpolate(
   A=VectorParameter() {group=VHCapsule, parameterName=CapsuleBGOutterColor, defaultValue=<rgba 0.0383 0 0.724 1>}.RGB
   B=VectorParameter() {group=VHCapsule, parameterName=CapsuleBGInnerColor, defaultValue=<rgba 0.276 0.8093 1 1>}.RGB
   Alpha=Multiply(A=Saturate(None=@n1), B=Clamp(None=@n3, Max=@n2) {maxDefault=1.5, minDefault=1.5})
  )
  n5 = NamedRerouteUsage()
  n6 = Saturate(None=Add(A=@n5, B=NamedRerouteUsage()))
  n7 = VectorParameter() {group=VHCapsule, parameterName=Static2Color, defaultValue=<rgba 1 0.0181 0 1>}
  n8 = VectorParameter() {group=VHCapsule, parameterName=StaticColor, defaultValue=<rgba 1.784 3.063 10 1>}
  n9 = StaticSwitchParameter(True=Multiply(A=@n6, B=LinearInterpolate(A=@n7.RGB, B=@n8.RGB, Alpha=@n5)), False=Multiply(A=@n6, B=@n8.RGB)) {group=VHCapsule, parameterName=UseSecondaryStaticColor}
  FunctionOutput(None=SetMaterialAttributes!(
    MaterialAttributes=FunctionInput() {inputType=FunctionInput_MaterialAttributes}
    EmissiveColor=LinearInterpolate(A=@n4, B=@n9, Alpha=@n6)
    Roughness=ScalarParameter() {defaultValue=0.15, group=VHCapsule, parameterName=Roughness}
    BaseColor=Constant()
   ) {attributeSetTypes [#1, #2, #3]}) {outputName=Emissive}
  n10 = Constant3Vector() {constant=<rgba 0.945 0.825 0 1>}
  n11 = Subtract(
   A=TextureCoordinate()
   B=FunctionInput(Preview=ComponentMask(None=@n10) {g=true, r=true}) {bUsePreviewValueAsDefault=true, inputName=UVPosition, inputType=FunctionInput_Vector2}
  ) {constB=0.5}
  n12 = ComponentMask(None=@n11) {g=true}
  NamedRerouteDeclaration(None=SmoothStep(Value=Abs(None=@n12)) {constMax=0.075}) {name=CapsuleSideGradient, nodeColor=<rgba 1 0 0.8672 1>}
  n13 = Multiply(A=Time(), B=ScalarParameter() {defaultValue=0.25, group=VHCapsule, parameterName=FlickerSpeed})
  n14 = ConstantBiasScale(None=Sine(None=Multiply(A=@n13) {constB=1.5}))
  n15 = ConstantBiasScale(None=Sine(None=Multiply(A=@n13) {constB=0.8}) {period=1.7})
  NamedRerouteDeclaration(None=Saturate(None=Add(A=@n14, B=@n15))) {name=Flicker, nodeColor=<rgba 0.2891 1 0 1>}
  n16 = Multiply(A=ScalarParameter() {defaultValue=10, group=VHCapsule, parameterName=StaticScale}, B=Constant2Vector() {g=0.75, r=1})
  n17 = MaterialFunctionCall!("Scale(10)"=@n16, "Rotation(0-1)", TimeOffset=NamedRerouteUsage(), Depth=Constant() {r=2}) {materialFunction=&/MF_VoltaicHuntress_Static}
  NamedRerouteDeclaration(None=Multiply(A=@n17.Static, B=NamedRerouteUsage())) {name=Static1, nodeColor=<rgba 0 0.3906 1 1>}
  n18 = NamedRerouteUsage()
  n19 = NamedRerouteUsage()
  n20 = OneMinus(None=Length(None=Multiply(A=@n18, B=@n19)) {is=class MaterialExpressionLength})
  n21 = Power(
   Base=Saturate(None=Power(Base=@n20) {constExponent=64})
   Exp=ScalarParameter() {defaultValue=6.1328, group=VHCapsule, parameterName=StaticMaskPower}
  ) {constExponent=1}
  n22 = ScalarParameter() {defaultValue=5, group=VHCapsule, parameterName=StaticMaskIntensity}
  NamedRerouteDeclaration(None=Saturate(None=Multiply(A=@n21, B=@n22))) {name=Static Mask, nodeColor=<rgba 0.9688 0 1 1>}
  n23 = BumpOffset(
   Coordinate=NamedRerouteDeclaration(None=@n11) {name=UV, nodeColor=<rgba 1 0.3516 0 1>}
   Height=ScalarParameter() {defaultValue=1, group=VHCapsule, parameterName=FakeDepthAmount}
  ) {heightRatio=-0.1}
  n24 = NamedRerouteDeclaration(None=ComponentMask(None=VectorParameter() {group=VHCapsule, parameterName=FakeDepthScale, defaultValue=<rgba 0.5 0.5 0 1>}.RGB) {g=true, r=true}) {name=FakeDepthScale, nodeColor=<rgba 0.125 1 0 1>}
  n25 = OneMinus(None=Length(None=Multiply(A=@n23, B=@n24)))
  NamedRerouteDeclaration(None=Saturate(None=Power(Base=@n25) {constExponent=64})) {name=CylinderFakeDepth, nodeColor=<rgba 0 0.5547 1 1>}
  NamedRerouteDeclaration(None=FunctionInput(Preview=Constant() {r=3}) {bUsePreviewValueAsDefault=true, inputName=TimeOffsetMask, inputType=FunctionInput_Scalar}) {name=TimeOffset, nodeColor=<rgba 0.0547 0 1 1>}
  n26 = ScalarParameter() {defaultValue=1, group=VHCapsule, parameterName=StaticBGOpacity}
  n27 = MaterialFunctionCall("Scale(10)"=Multiply(A=@n16) {constB=0.9}, TimeOffset=Add(A=Constant() {r=3.75}, B=NamedRerouteUsage()), Depth=Constant() {r=5}) {materialFunction=&MF_VoltaicHuntress_Static}
  NamedRerouteDeclaration(None=Multiply(A=Multiply(A=@n26, B=@n27.Static) {constB=0.75}, B=NamedRerouteUsage())) {name=Static2, nodeColor=<rgba 1 0.3516 0 1>}
  SmoothStep(Value=@n25) {constMax=2, constMin=0.9}
 ]}
