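# Parent character material for the Voltaic Huntress player outfit, built between the shared 2023 PreFX and PostFX functions.
# Switchable accessory, capsule and blade functions blend in emissive effects using parameterized UV circle and box masks.
# An optional per-capsule backpack mask quantizes pre-skinned position into steps.
# Ice, galaxy and holographic glitter style layers follow, with an isGold switch adding carbon fiber and its clear coat normal.
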Material /BRCosmetics/Characters/Player/Female/Medium/Bodies/F_MED_VoltaicHuntress/Materials/M_VoltaicHuntress_Parent
graph {
 properties {
  bUseMaterialAttributes=true
  bUsedWithClothing=true
  bUsedWithMeshDeformer=true
  bUsedWithMorphTargets=true
  bUsedWithNanite=true
  bUsedWithNiagaraMeshParticles=true
  bUsedWithSkeletalMesh=true
  pixelDepthOffsetMode=PDOM_Legacy
  shadingModel=MSM_FromMaterialExpression
  subsurfaceProfile=/Game/Packages/Fortress_SharedMaterials/SubSurfaceProfile/Character_SubsurfaceProfile
  usedShadingModels=Default Lit | Subsurface | Clear Coat | Subsurface Profile | Cloth
 }
 nodes [
  n1 = MaterialFunctionCall() {materialFunction=&MF_CharacterParent_2023_PreFX}
  n2 = MaterialFunctionCall!(
   Tech Art Mask=@n1.Tech Art Mask
   Mask=@n1.Tech Art Mask
   In=MaterialFunctionCall!(TechArt=@n1.Tech Art Mask, Fresnel=@n1.Tech Art Mask, In=NamedRerouteUsage()) {materialFunction=/Game/Global/Materials/Character_Parent_2023/CrewStyle_Parent_2023/Ice/MF_CrewStyle_Ice_Inst}.Result
  ) {materialFunction=/Game/Global/Materials/Character_Parent_2023/CrewStyle_Parent_2023/Galaxy/MF_CrewStyle_Galaxy_Inst}
  n3 = MaterialFunctionCall!(Alpha=@n1.Tech Art Mask, In=@n2.Result) {materialFunction=/Game/Global/MaterialFunctions/TechArt/MF_CarbonFiberDualLayer}
  n4 = GetMaterialAttributes(None=@n3.Result) {attributeGetTypes [#1]}
  n5 = MaterialFunctionCall(
   BaseNormal=@n4.Normal
   AdditionalNormal=TextureSample(UVs=@n3.CarbonFiberUvs) {samplerType=SAMPLERTYPE_Normal, texture=/Game/Weapons/WeaponSkins/Wraps/Textures/T_Wrap_CarbonFiber_N}.RGB
  ) {materialFunction=&BlendAngleCorrectedNormals}
  ClearCoatNormalCustomOutput!(None=StaticSwitchParameter(True=LinearInterpolate(A=@n4.Normal, B=@n5.Result, Alpha=@n3.Mask), False=@n4.Normal) {parameterName=isGold})
  MaterialFunctionCall(
   Material=MaterialFunctionCall!(Alpha=@n1.Tech Art Mask, In=StaticSwitchParameter(True=@n3.Result, False=@n2.Result) {parameterName=isGold}, Fresnel=@n1.Tech Art Mask) {materialFunction=/Game/Global/MaterialFunctions/TechArt/MF_GlitterHolographic}.Result
   M=@n1.M
   Pre Skinned Local Position=@n1.M
   Distance Blend=@n1.M
   Fresnel=@n1.M
  ) {materialFunction=&MF_CharacterParent_2023_PostFX}
  NamedRerouteDeclaration(None=@n1.Result) {name=InMaterialAttributes, nodeColor=<rgba 0 0.7891 1 1>}
  NamedRerouteDeclaration(None=@n1.Tech Art Mask) {name=InFXMask, nodeColor=<rgba 1 0.1875 0 1>}
  n6 = Add(A=VertexInterpolator(VS=PreSkinnedPosition()).PS, B=Constant3Vector() {constant=<rgba 0 10.85 0 0>}) {constB=10.5}
  n7 = Constant3Vector() {constant=<rgba 1 0 0 0>}
  n8 = Constant() {r=-0.011}
  n9 = Normalize(VectorInput=@n6)
  n10 = ComponentMask(None=Add(A=@n6, B=RotateAboutAxis(NormalizedRotationAxis=@n7, RotationAngle=@n8, PivotPoint=@n9, Position=@n6))) {g=true}
  n11 = Round(None=Multiply(A=Divide(A=@n10) {constB=10}) {constB=4})
  NamedRerouteDeclaration(None=StaticSwitchParameter(True=Multiply(A=@n11) {constB=0.25}, False=Constant()) {group=VHCapsule, parameterName=UseBackpackIndividualCapsuleMask}) {name=BackpackIndividualCapsuleMask, nodeColor=<rgba 0.8516 0 1 1>}
  n12 = NamedRerouteUsage()
  n13 = MaterialFunctionCall!(In=@n12, UVPositionOffset=NamedRerouteUsage()) {materialFunction=/BRCosmetics/Characters/Player/Female/Medium/Bodies/F_MED_VoltaicHuntress/Materials/MaterialFunctions/MF_VoltaicHuntress_Accessory}
  n14 = BlendMaterialAttributes(A=@n12, B=@n13.Emissive, Alpha=NamedRerouteUsage())
  n15 = MaterialFunctionCall(In=@n14, UVPositionOffset=NamedRerouteUsage()) {materialFunction=&MF_VoltaicHuntress_Accessory}
  n16 = StaticSwitchParameter(True=BlendMaterialAttributes(A=@n14, B=@n15.Emissive, Alpha=NamedRerouteUsage()), False=@n14) {parameterName=UseSecondAccessory}
  n17 = MaterialFunctionCall!(In=@n16, TimeOffsetMask=NamedRerouteUsage(), UVPosition=NamedRerouteUsage()) {materialFunction=/BRCosmetics/Characters/Player/Female/Medium/Bodies/F_MED_VoltaicHuntress/Materials/MaterialFunctions/MF_VoltaicHuntress_Capsule}
  n18 = StaticSwitchParameter(True=BlendMaterialAttributes(A=@n16, B=@n17.Emissive, Alpha=NamedRerouteUsage()), False=@n16) {parameterName=UseCapsule}
  n19 = MaterialFunctionCall!(In=@n18) {materialFunction=/BRCosmetics/Characters/Player/Female/Medium/Bodies/F_MED_VoltaicHuntress/Materials/MaterialFunctions/MF_VoltaicHuntress_Blade}
  n20 = NamedRerouteUsage()
  NamedRerouteDeclaration(None=StaticSwitchParameter(True=BlendMaterialAttributes(A=@n18, B=@n19.Emissive, Alpha=@n20), False=@n18) {parameterName=UseBlade}) {name=OutMaterialAttributes, nodeColor=<rgba 0 1 0.1562 1>}
  n21 = TextureCoordinate()
  n22 = ComponentMask(None=VectorParameter() {group=VHAccessory, parameterName=AccessoryMaskPosition, defaultValue=<rgba 0.335 0.07 0 1>}.RGB) {g=true, r=true}
  n23 = Multiply(
   A=Subtract(A=Subtract(A=@n21) {constB=0.5}, B=NamedRerouteDeclaration(None=@n22) {name=CircleUVPosition, nodeColor=<rgba 0 1 0.8359 1>}) {constB=0}
   B=ComponentMask(None=VectorParameter() {group=VHAccessory, parameterName=AccessoryMaskScaleOverride, defaultValue=<rgba 1 1 0 1>}.RGB) {g=true, r=true}
  )
  NamedRerouteDeclaration(None=Step(
    Y=Length(None=@n23) {is=class MaterialExpressionLength}
    X=ScalarParameter() {defaultValue=0.039, group=VHAccessory, parameterName=AccessoryMaskScale}
   ) {constX=0.1}) {name=BeltCircleMask, nodeColor=<rgba 0 1 0.0391 1>}
  n24 = ComponentMask(None=VectorParameter() {group=VHCapsule, parameterName=CylinderMaskPosition, defaultValue=<rgba 0.942 0.83 0 1>}.RGB) {g=true, r=true}
  n25 = Constant() {r=0.05}
  n26 = ComponentMask(None=VectorParameter() {group=VHCapsule, parameterName=CylinderMaskScale, defaultValue=<rgba 2.35 1.2 0 1>}.RGB) {g=true, r=true}
  NamedRerouteDeclaration(None=MaterialFunctionCall!(
    A=TextureCoordinate()
    B=NamedRerouteDeclaration(None=@n24) {name=CylinderUVPosition, nodeColor=<rgba 0 0.6953 1 1>}
    Bounds=Multiply(A=@n25, B=@n26)
    Edge Falloff=Constant()
   ) {materialFunction=/Engine/Functions/Engine_MaterialFunctions01/Gradient/BoxMask-2D}.Result) {name=BeltCylinderMask, nodeColor=<rgba 0 0.625 1 1>}
  n27 = VectorParameter() {group=VHAccessory, parameterName=AccessoryMaskPosition2, defaultValue=<rgba 0.335 0.07 0 1>}
  n28 = Subtract(
   A=Subtract(A=TextureCoordinate()) {constB=0.5}
   B=NamedRerouteDeclaration(None=ComponentMask(None=@n27.RGB) {g=true, r=true}) {name=Circle2UVPosition, nodeColor=<rgba 0.0458 0.0257 1 1>}
  ) {constB=0}
  NamedRerouteDeclaration(None=Step(Y=Length(None=@n28), X=ScalarParameter() {defaultValue=0.039, group=VHAccessory, parameterName=AccessoryMaskScale}) {constX=0.1}) {name=BeltCircleMask2, nodeColor=<rgba 0.7877 0.2746 1 1>}
  n29 = VectorParameter() {group=VHBlade, parameterName=BladeMaskPosition, defaultValue=<rgba 0.942 0.83 0 1>}
  n30 = Constant() {r=0.05}
  n31 = ComponentMask(None=VectorParameter() {group=VHBlade, parameterName=BladeMaskScale, defaultValue=<rgba 2.35 1.2 0 1>}.RGB) {g=true, r=true}
  NamedRerouteDeclaration(None=MaterialFunctionCall(A=TextureCoordinate(), B=ComponentMask(None=@n29.RGB) {g=true, r=true}, Bounds=Multiply(A=@n30, B=@n31), Edge Falloff=Constant()) {materialFunction=&BoxMask-2D}.Result) {name=BladeMask, nodeColor=<rgba 0 0.625 1 1>}
 ]
}
