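# Control Rig for the Voltaic Huntress character body that poses the skirt and hip accessory bones procedurally from leg motion and player state.
# Spherical pose readers, remapped thigh and calf angles and spine height are stored as alpha variables.
# These weight chained parent constraints to thighs and calves plus corrective transform offsets, often smoothed with AlphaInterp.
# Skydiving, swimming, knee-slide, DBNO and frontend flags add or suppress adjustments.
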ControlRigBlueprint /BRCosmetics/Characters/Player/Female/Medium/Bodies/F_MED_VoltaicHuntress/Meshes/F_MED_VoltaicHuntress_CtrlRig
graphs [
 RigVMFunctionLibrary {links [], nodes []}
 RigVMModel {nodes [
   RigUnit_BeginExecution = &RigUnit_BeginExecution()
   n1 = &RigUnit_SphericalPoseReader(
    ActiveRegionScaleFactors=(PositiveWidth=4,NegativeWidth=3.5,PositiveHeight=2,NegativeHeight=3)
    ActiveRegionSize=0.1
    Debug=(bDrawDebug=false,bDraw2D=False,bDrawLocalAxes=False,DebugScale=25,DebugSegments=20,DebugThickness=0.25)
    DriverAxis=(X=-1,Y=0,Z=0)
    DriverItem=(Type=Bone,Name="thigh_l")
    FalloffRegionScaleFactors=(PositiveWidth=1,NegativeWidth=1,PositiveHeight=1,NegativeHeight=1)
    FalloffSize=0.45
    FlipHeightScaling=False
    FlipWidthScaling=False
    OptionalParentItem=(Type=Bone,Name="None")
    RotationOffset=(X=130,Y=-10,Z=0)
    ExecutePin=@RigUnit_BeginExecution.ExecutePin
   )
   > n2 = &RigUnit_SphericalPoseReader(
    ActiveRegionScaleFactors=(PositiveWidth=4,NegativeWidth=3.5,PositiveHeight=2,NegativeHeight=3)
    ActiveRegionSize=0.1
    Debug=(bDrawDebug=false,bDraw2D=False,bDrawLocalAxes=False,DebugScale=25,DebugSegments=20,DebugThickness=0.25)
    DriverAxis=(X=1,Y=0,Z=0)
    DriverItem=(Type=Bone,Name="thigh_r")
    FalloffRegionScaleFactors=(PositiveWidth=1,NegativeWidth=1,PositiveHeight=1,NegativeHeight=1)
    FalloffSize=0.45
    FlipHeightScaling=False
    FlipWidthScaling=true
    OptionalParentItem=(Type=Bone,Name="None")
    RotationOffset=(X=-50,Y=-10,Z=0)
   )
   RigVMFunction_MathQuaternionToEuler = &RigVMFunction_MathQuaternionToEuler(
    RotationOrder=ZYX
    Value=&RigUnit_GetRelativeTransformForItem(Child=(Type=Bone,Name="calf_l"), Parent=(Type=Bone,Name="thigh_l"), bChildInitial=False, bParentInitial=False).RelativeTransform.Rotation
   )
   Remap = &RigVMFunction_MathDoubleRemap(SourceMaximum=70, SourceMinimum=20, TargetMaximum=1, TargetMinimum=0, Value=@RigVMFunction_MathQuaternionToEuler.Result.Z)
   > RigVMVariableNode(Value=@Remap.Result) {variable=L calf alpha}
   n3 = &RigUnit_GetRelativeTransformForItem(Child=(Type=Bone,Name="calf_r"), Parent=(Type=Bone,Name="thigh_r"), bChildInitial=False, bParentInitial=False)
   > RigVMVariableNode(Value=&RigVMFunction_MathDoubleRemap(
     SourceMaximum=70
     SourceMinimum=20
     TargetMaximum=1
     TargetMinimum=0
     Value=&RigVMFunction_MathQuaternionToEuler(RotationOrder=ZYX, Value=@n3.RelativeTransform.Rotation).Result.Z
    ).Result) {variable=R calf alpha}
   n4 = &RigUnit_GetRelativeTransformForItem(Child=(Type=Bone,Name="thigh_l"), Parent=(Type=Bone,Name="pelvis"), bChildInitial=False, bParentInitial=False)
   > RigVMVariableNode(Value=&RigVMFunction_MathDoubleRemap(
     SourceMaximum=-90
     SourceMinimum=-50
     TargetMaximum=1
     TargetMinimum=0
     Value=&RigVMFunction_MathQuaternionToEuler(RotationOrder=ZYX, Value=@n4.RelativeTransform.Rotation).Result.Z
    ).Result) {variable=L thigh VeryForward}
   > RigVMVariableNode(Value=&RigVMFunction_MathDoubleMul(B=-1, A=&RigVMFunction_MathQuaternionToEuler(RotationOrder=YXZ, Value=@n4.RelativeTransform.Rotation).Result.Z).Result) {variable=L_thigh_Z_value}
   n5 = &RigUnit_GetRelativeTransformForItem(Child=(Type=Bone,Name="thigh_r"), Parent=(Type=Bone,Name="pelvis"), bChildInitial=False, bParentInitial=False)
   Absolute = &RigVMFunction_MathDoubleAbs(Value=&RigVMFunction_MathQuaternionToEuler(RotationOrder=ZYX, Value=@n5.RelativeTransform.Rotation).Result.Z)
   > RigVMVariableNode(Value=&RigVMFunction_MathDoubleRemap(SourceMaximum=90, SourceMinimum=150, TargetMaximum=1, TargetMinimum=0, Value=@Absolute.Result).Result) {variable=R thigh VeryForward}
   n6 = &RigVMFunction_MathQuaternionToEuler(RotationOrder=YXZ, Value=@n5.RelativeTransform.Rotation)
   > RigVMVariableNode(Value=&RigVMFunction_MathDoubleSub(A=180, B=&RigVMFunction_MathDoubleAbs(Value=@n6.Result.Z).Result).Result) {variable=R_thigh_Z_value}
   n7 = &RigUnit_GetRelativeTransformForItem(Child=(Type=Bone,Name="spine_01"), Parent=(Type=Bone,Name="root"), bChildInitial=False, bParentInitial=false)
   > RigVMVariableNode(Value=&RigVMFunction_MathDoubleRemap(SourceMaximum=54, SourceMinimum=83, TargetMaximum=1, TargetMinimum=0, Value=@n7.RelativeTransform.Translation.Z).Result) {variable=spineTranslationAlpha Lower}
   > RigVMVariableNode(Value=&RigVMFunction_MathDoubleRemap(SourceMaximum=77, SourceMinimum=100, TargetMaximum=1, TargetMinimum=0, Value=@n7.RelativeTransform.Translation.Z).Result) {variable=spineTranslationAlpha Higher}
   n8 = &RigUnit_GetRelativeTransformForItem(Child=(Type=Bone,Name="thigh_l"), Parent=(Type=Bone,Name="pelvis"), bChildInitial=False, bParentInitial=False)
   > RigVMVariableNode(Value=&RigVMFunction_MathDoubleRemap(
     SourceMaximum=-20
     SourceMinimum=0
     TargetMaximum=1
     TargetMinimum=0
     Value=&RigVMFunction_MathQuaternionToEuler(RotationOrder=XYZ, Value=@n8.RelativeTransform.Rotation).Result.Y
    ).Result) {variable=L thighOut delta}
   n9 = &RigUnit_GetRelativeTransformForItem(Child=(Type=Bone,Name="thigh_r"), Parent=(Type=Bone,Name="pelvis"), bChildInitial=False, bParentInitial=False)
   > RigVMVariableNode(Value=&RigVMFunction_MathDoubleRemap(
     SourceMaximum=-20
     SourceMinimum=0
     TargetMaximum=1
     TargetMinimum=0
     Value=&RigVMFunction_MathQuaternionToEuler(RotationOrder=XYZ, Value=@n9.RelativeTransform.Rotation).Result.Y
    ).Result) {variable=R thighOut delta}
   > RigVMVariableNode(Value=@n1.OutputParam) {variable=L thigh forward}
   > RigVMVariableNode(Value=@n2.OutputParam) {variable=R thigh forward}
   n10 = RigVMVariableNode() {variable=L_thigh_Z_value}
   n11 = RigVMVariableNode() {variable=R_thigh_Z_value}
   n12 = &RigVMFunction_AlphaInterp(
    InterpSpeedDecreasing=20
    InterpSpeedIncreasing=20
    Scale=1
    bClampResult=False
    Value=RigVMDispatchNode(
     False=1
     True=0
     Condition=/Script/RigVM.RigVMFunction_MathDoubleGreater(B=@n10.Value, A=@n11.Value) {is=struct RigVMUnitNode, template="Greater::Execute(in A,in B,out Result)"}.Result
    ) {template="DISPATCH_RigVMDispatch_If(in Condition,in True,in False,out Result)"}.Result
   )
   n13 = &RigVMFunction_MathDoubleAdd(A=@n12.Result, B=RigVMVariableNode() {variable=L thigh VeryForward}.Value)
   > &RigUnit_ParentConstraint(
    AdvancedSettings=(InterpolationType=Average,RotationOrderForFilter=XZY)
    Child=(Type=Bone,Name="dyn_L_skirtDriver_1")
    Filter=(TranslationFilter=(bX=true,bY=true,bZ=true),RotationFilter=(bX=true,bY=true,bZ=True),ScaleFilter=(bX=True,bY=True,bZ=True))
    Parents=((Item=(Type=Bone,Name="thigh_l"),Weight=1))
    Weight=@n13.Result
   )
   n14 = &RigVMFunction_MathDoubleMul(B=0.6, A=RigVMVariableNode() {variable=R thighOut delta}.Value)
   n15 = RigVMVariableNode() {variable=R_thigh_Z_value}
   Greater = &RigVMFunction_MathDoubleGreater(A=RigVMVariableNode() {variable=L_thigh_Z_value}.Value, B=@n15.Value) {template="Greater::Execute(in A,in B,out Result)"}
   DISPATCH_RigVMDispatch_If = RigVMDispatchNode(False=1, True=0, Condition=@Greater.Result) {template="DISPATCH_RigVMDispatch_If(in Condition,in True,in False,out Result)"}
   n16 = &RigVMFunction_AlphaInterp(InterpSpeedDecreasing=20, InterpSpeedIncreasing=20, Scale=1, bClampResult=False, Value=@DISPATCH_RigVMDispatch_If.Result)
   n17 = &RigVMFunction_MathDoubleSub(B=RigVMVariableNode() {variable=L thigh VeryForward}.Value, A=@n16.Result)
   > &RigUnit_ParentConstraint(
    AdvancedSettings=(InterpolationType=Average,RotationOrderForFilter=XZY)
    Child=(Type=Bone,Name="dyn_L_skirtDriver_1")
    Filter=(TranslationFilter=(bX=true,bY=false,bZ=true),RotationFilter=(bX=true,bY=true,bZ=true),ScaleFilter=(bX=True,bY=True,bZ=True))
    Parents=((Item=(Type=Bone,Name="thigh_r"),Weight=1))
    Weight=&RigVMFunction_MathDoubleSub(A=&RigVMFunction_MathDoubleAdd(A=@n14.Result, B=@n17.Result).Result, B=@n13.Result).Result
   )
   > &RigUnit_ModifyTransforms(
    ItemToModify=((Item=(Type=Bone,Name="dyn_L_side_skirt_1"),Transform=(Rotation=(X=-0.0689,Y=-0.6207,Z=-0.2915,W=0.7245),Translation=(X=0,Y=2,Z=0),Scale3D=(X=1,Y=1,Z=1))))
    WeightMinimum=0
    Weight=RigVMVariableNode() {variable=L thigh VeryForward}.Value
   )
   > &RigUnit_ModifyTransforms(
    ItemToModify=((Item=(Type=Bone,Name="dyn_L_side_skirt_1"),Transform=(Rotation=(X=0,Y=0,Z=-0.0872,W=0.9962),Translation=(X=0.5,Y=0.25,Z=0),Scale3D=(X=1,Y=1,Z=1))))
    WeightMinimum=0
    Weight=RigVMVariableNode() {variable=L thighOut delta}.Value
   )
   n18 = RigVMVariableNode() {variable=L thigh forward}
   > &RigUnit_ParentConstraint(
    AdvancedSettings=(InterpolationType=Average,RotationOrderForFilter=XZY)
    Child=(Type=Bone,Name="dyn_L_back_skirt_1")
    Filter=(TranslationFilter=(bX=true,bY=true,bZ=true),RotationFilter=(bX=True,bY=True,bZ=True),ScaleFilter=(bX=True,bY=True,bZ=True))
    Parents=((Item=(Type=Bone,Name="thigh_l"),Weight=1))
    Weight=&RigVMFunction_MathDoubleSub(A=&RigVMFunction_MathDoubleSub(A=1, B=@n18.Value).Result, B=RigVMVariableNode() {variable=spineTranslationAlpha Lower}.Value).Result
   )
   Add = &RigVMFunction_MathDoubleAdd(
    A=&RigVMFunction_MathBoolToFloat(Value=RigVMVariableNode() {variable=isSkydiving}.Value).Result
    B=&RigVMFunction_MathBoolToFloat(Value=RigVMVariableNode() {variable=isSwimming}.Value).Result
   )
   n19 = RigVMVariableNode() {variable=L thigh VeryForward}
   > &RigUnit_ParentConstraint(
    AdvancedSettings=(InterpolationType=Average,RotationOrderForFilter=XZY)
    Child=(Type=Bone,Name="dyn_L_back_skirt_1")
    Filter=(TranslationFilter=(bX=true,bY=true,bZ=true),RotationFilter=(bX=True,bY=True,bZ=True),ScaleFilter=(bX=True,bY=True,bZ=True))
    Parents=((Item=(Type=Bone,Name="thigh_l"),Weight=1))
    Weight=&RigVMFunction_MathDoubleMul(A=@Add.Result, B=&RigVMFunction_MathDoubleSub(A=1, B=@n19.Value).Result).Result
   )
   n20 = &RigVMFunction_MathDoubleSub(A=1, B=RigVMVariableNode() {variable=L thigh forward}.Value)
   n21 = &RigVMFunction_AlphaInterp(InterpSpeedDecreasing=20, InterpSpeedIncreasing=20, Scale=1, bClampResult=False, Value=@n20.Result)
   > &RigUnit_ParentConstraint(
    AdvancedSettings=(InterpolationType=Average,RotationOrderForFilter=XZY)
    Child=(Type=Bone,Name="dyn_L_back_skirt_2")
    Filter=(TranslationFilter=(bX=false,bY=false,bZ=false),RotationFilter=(bX=True,bY=True,bZ=True),ScaleFilter=(bX=True,bY=True,bZ=True))
    Parents=((Item=(Type=Bone,Name="calf_l"),Weight=1))
    Weight=@n21.Result
   )
   n22 = RigVMVariableNode() {variable=spineTranslationAlpha Lower}
   n23 = &RigVMFunction_MathBoolToFloat(Value=RigVMVariableNode() {variable=isDBNO}.Value)
   n24 = RigVMVariableNode() {variable=L calf alpha}
   > &RigUnit_ModifyTransforms(
    ItemToModify=((Item=(Type=Bone,Name="dyn_L_back_skirt_1"),Transform=(Rotation=(X=0,Y=0,Z=-0.1736,W=0.9848),Translation=(X=0,Y=0,Z=0),Scale3D=(X=1,Y=1,Z=1))))
    WeightMinimum=0
    Weight=&RigVMFunction_MathDoubleSub(
     B=&RigVMFunction_MathDoubleMul(B=0.5, A=@n22.Value).Result
     A=&RigVMFunction_MathDoubleSub(B=@n23.Result, A=@n24.Value).Result
    ).Result
   )
   n25 = RigVMVariableNode() {variable=isSwimming}
   n26 = &RigVMFunction_MathDoubleSub(
    A=&RigVMFunction_MathDoubleSub(A=1, B=RigVMVariableNode() {variable=R thighOut delta}.Value).Result
    B=RigVMVariableNode() {variable=R thigh forward}.Value
   )
   n27 = RigVMVariableNode() {variable=L thigh forward}
   n28 = &RigVMFunction_MathDoubleSub(
    B=&RigVMFunction_MathBoolToFloat(Value=RigVMVariableNode() {variable=isInFrontend}.Value).Result
    A=&RigVMFunction_MathDoubleSub(
     B=&RigVMFunction_MathBoolToFloat(Value=@n25.Value).Result
     A=&RigVMFunction_MathDoubleSub(A=@n26.Result, B=@n27.Value).Result
    ).Result
   )
   > &RigUnit_ModifyTransforms(
    ItemToModify=((Item=(Type=Bone,Name="dyn_L_back_skirt_1"),Transform=(Rotation=(X=0.1579,Y=0.3034,Z=-0.4339,W=0.8335),Translation=(X=0,Y=0,Z=0),Scale3D=(X=1,Y=1,Z=1))))
    WeightMinimum=0
    Weight=@n28.Result
   )
   > &RigUnit_ModifyTransforms(
    ItemToModify=((Item=(Type=Bone,Name="dyn_L_side_skirt_1"),Transform=(Rotation=(X=0,Y=0,Z=-0.1736,W=0.9848),Translation=(X=0,Y=0,Z=0),Scale3D=(X=1,Y=1,Z=1))))
    WeightMinimum=0
    Weight=@n28.Result
   )
   n29 = &RigVMFunction_MathDoubleAdd(B=45, A=RigVMVariableNode() {variable=R_thigh_Z_value}.Value)
   n30 = RigVMVariableNode() {variable=L_thigh_Z_value}
   AlphaInterp = &RigVMFunction_AlphaInterp(
    InterpSpeedDecreasing=40
    InterpSpeedIncreasing=40
    Scale=1
    bClampResult=False
    Value=RigVMDispatchNode(
     False=1
     True=0
     Condition=&RigVMFunction_MathDoubleGreater(B=@n29.Result, A=@n30.Value) {template="Greater::Execute(in A,in B,out Result)"}.Result
    ) {template="DISPATCH_RigVMDispatch_If(in Condition,in True,in False,out Result)"}.Result
   )
   n31 = &RigVMFunction_MathDoubleAdd(A=@AlphaInterp.Result, B=RigVMVariableNode() {variable=R thighOut delta}.Value)
   Maximum = &RigVMFunction_MathDoubleMax(B=0.85, A=@n31.Result)
   > &RigUnit_ParentConstraint(
    AdvancedSettings=(InterpolationType=Average,RotationOrderForFilter=XZY)
    Child=(Type=Bone,Name="dyn_front_skirt_1")
    Filter=(TranslationFilter=(bX=true,bY=true,bZ=true),RotationFilter=(bX=True,bY=True,bZ=True),ScaleFilter=(bX=True,bY=True,bZ=True))
    Parents=((Item=(Type=Bone,Name="thigh_r"),Weight=1))
    Weight=@Maximum.Result
   )
   n32 = RigVMVariableNode() {variable=L_thigh_Z_value}
   n33 = RigVMVariableNode() {variable=R_thigh_Z_value}
   n34 = &RigVMFunction_AlphaInterp(
    InterpSpeedDecreasing=40
    InterpSpeedIncreasing=40
    Scale=1
    bClampResult=False
    Value=RigVMDispatchNode(
     False=0.7
     True=0
     Condition=&RigVMFunction_MathDoubleGreater(B=@n32.Value, A=@n33.Value) {template="Greater::Execute(in A,in B,out Result)"}.Result
    ) {template="DISPATCH_RigVMDispatch_If(in Condition,in True,in False,out Result)"}.Result
   )
   n35 = RigVMVariableNode() {variable=L thigh forward}
   > &RigUnit_ParentConstraint(
    AdvancedSettings=(InterpolationType=Average,RotationOrderForFilter=XZY)
    Child=(Type=Bone,Name="dyn_front_skirt_1")
    Filter=(TranslationFilter=(bX=true,bY=false,bZ=true),RotationFilter=(bX=True,bY=true,bZ=True),ScaleFilter=(bX=True,bY=True,bZ=True))
    Parents=((Item=(Type=Bone,Name="thigh_l"),Weight=1))
    Weight=&RigVMFunction_MathDoubleSub(A=&RigVMFunction_MathDoubleMul(B=@n34.Result, A=@n35.Value).Result, B=@n31.Result).Result
   )
   > &RigUnit_ParentConstraint(
    AdvancedSettings=(InterpolationType=Average,RotationOrderForFilter=XZY)
    Child=(Type=Bone,Name="dyn_front_skirt_2")
    Filter=(TranslationFilter=(bX=true,bY=true,bZ=true),RotationFilter=(bX=True,bY=True,bZ=True),ScaleFilter=(bX=True,bY=True,bZ=True))
    Parents=((Item=(Type=Bone,Name="thigh_r"),Weight=1))
    Weight=0.8
   )
   > &RigUnit_ModifyTransforms(
    ItemToModify=((Item=(Type=Bone,Name="dyn_front_skirt_3"),Transform=(Rotation=(X=-0.0665,Y=0.1604,Z=0.3769,W=0.9098),Translation=(X=0,Y=0,Z=0),Scale3D=(X=1,Y=1,Z=1))))
    WeightMinimum=0
    Weight=RigVMVariableNode() {variable=R calf alpha}.Value
   )
   > &RigUnit_ParentConstraint(
    AdvancedSettings=(InterpolationType=Average,RotationOrderForFilter=XZY)
    Child=(Type=Bone,Name="dyn_R_side_skirt_2")
    Filter=(TranslationFilter=(bX=true,bY=true,bZ=true),RotationFilter=(bX=true,bY=true,bZ=true),ScaleFilter=(bX=True,bY=True,bZ=True))
    Parents=((Item=(Type=Bone,Name="calf_r"),Weight=1))
    Weight=0.4
   )
   > &RigUnit_ModifyTransforms(
    ItemToModify=((Item=(Type=Bone,Name="dyn_R_side_skirt_1"),Transform=(Rotation=(X=0,Y=0,Z=-0.1045,W=0.9945),Translation=(X=0,Y=0,Z=0),Scale3D=(X=1,Y=1,Z=1))))
    WeightMinimum=0
    Weight=RigVMVariableNode() {variable=R thighOut delta}.Value
   )
   > &RigUnit_ModifyTransforms(
    ItemToModify=((Item=(Type=Bone,Name="dyn_R_side_skirt_2"),Transform=(Rotation=(X=0.13,Y=-0.0114,Z=-0.0864,W=0.9877),Translation=(X=0,Y=0,Z=0),Scale3D=(X=1,Y=1,Z=1))))
    WeightMinimum=0
    Weight=RigVMVariableNode() {variable=R calf alpha}.Value
   )
   Subtract = &RigVMFunction_MathDoubleSub(A=1, B=RigVMVariableNode() {variable=R thigh forward}.Value)
   > &RigUnit_ParentConstraint(
    AdvancedSettings=(InterpolationType=Average,RotationOrderForFilter=XZY)
    Child=(Type=Bone,Name="dyn_R_back_skirt_1")
    Filter=(TranslationFilter=(bX=true,bY=true,bZ=true),RotationFilter=(bX=True,bY=True,bZ=True),ScaleFilter=(bX=True,bY=True,bZ=True))
    Parents=((Item=(Type=Bone,Name="thigh_r"),Weight=1))
    Weight=@Subtract.Result
   )
   n36 = RigVMVariableNode() {variable=isSkydiving}
   n37 = RigVMVariableNode() {variable=isSwimming}
   n38 = &RigVMFunction_MathDoubleMul(
    A=&RigVMFunction_MathDoubleAdd(
     A=&RigVMFunction_MathBoolToFloat(Value=@n36.Value).Result
     B=&RigVMFunction_MathBoolToFloat(Value=@n37.Value).Result
    ).Result
    B=&RigVMFunction_MathDoubleSub(A=1, B=RigVMVariableNode() {variable=R thigh VeryForward}.Value).Result
   )
   > &RigUnit_ParentConstraint(
    AdvancedSettings=(InterpolationType=Average,RotationOrderForFilter=XZY)
    Child=(Type=Bone,Name="dyn_R_back_skirt_1")
    Filter=(TranslationFilter=(bX=true,bY=true,bZ=true),RotationFilter=(bX=True,bY=True,bZ=True),ScaleFilter=(bX=True,bY=True,bZ=True))
    Parents=((Item=(Type=Bone,Name="thigh_r"),Weight=1))
    Weight=@n38.Result
   )
   > &RigUnit_ParentConstraint(
    AdvancedSettings=(InterpolationType=Average,RotationOrderForFilter=XZY)
    Child=(Type=Bone,Name="dyn_R_back_skirt_2")
    Filter=(TranslationFilter=(bX=false,bY=false,bZ=false),RotationFilter=(bX=True,bY=True,bZ=True),ScaleFilter=(bX=True,bY=True,bZ=True))
    Parents=((Item=(Type=Bone,Name="calf_r"),Weight=1))
    Weight=&RigVMFunction_MathDoubleMul(B=0.4, A=@n38.Result).Result
   )
   n39 = RigVMVariableNode() {variable=spineTranslationAlpha Higher}
   n40 = &RigVMFunction_MathBoolToFloat(Value=RigVMVariableNode() {variable=isDBNO}.Value)
   n41 = RigVMVariableNode() {variable=R calf alpha}
   > &RigUnit_ModifyTransforms(
    ItemToModify=((Item=(Type=Bone,Name="dyn_R_back_skirt_1"),Transform=(Rotation=(X=0,Y=0,Z=-0.342,W=0.9397),Translation=(X=0,Y=0,Z=0),Scale3D=(X=1,Y=1,Z=1))))
    WeightMinimum=0
    Weight=&RigVMFunction_MathDoubleSub(
     B=&RigVMFunction_MathDoubleMul(B=0.8, A=@n39.Value).Result
     A=&RigVMFunction_MathDoubleSub(B=@n40.Result, A=@n41.Value).Result
    ).Result
   )
   n42 = RigVMVariableNode() {variable=R thigh VeryForward}
   n43 = &RigVMFunction_AlphaInterp(InterpSpeedDecreasing=20, InterpSpeedIncreasing=20, Scale=0.5, bClampResult=False, Value=@n42.Value)
   > &RigUnit_ParentConstraint(
    AdvancedSettings=(InterpolationType=Average,RotationOrderForFilter=XZY)
    Child=(Type=Bone,Name="dyn_R_back_skirt_1")
    Filter=(TranslationFilter=(bX=true,bY=true,bZ=true),RotationFilter=(bX=true,bY=true,bZ=true),ScaleFilter=(bX=True,bY=True,bZ=True))
    Parents=((Item=(Type=Bone,Name="dyn_L_back_skirt_1"),Weight=0.5),(Item=(Type=Bone,Name="thigh_r"),Weight=0.5))
    Weight=@n43.Result
   )
   n44 = &RigVMFunction_AlphaInterp(InterpSpeedDecreasing=20, InterpSpeedIncreasing=20, Scale=1, bClampResult=False, Value=@n42.Value)
   > &RigUnit_ModifyTransforms(
    ItemToModify=((Item=(Type=Bone,Name="dyn_R_back_skirt_1"),Transform=(Rotation=(X=-0.0151,Y=0.0858,Z=0.173,W=0.9811),Translation=(X=0,Y=0,Z=0),Scale3D=(X=1,Y=1,Z=1))))
    WeightMinimum=0
    Weight=@n44.Result
   )
   n45 = RigVMVariableNode() {variable=L thighOut delta}
   n46 = &RigVMFunction_MathDoubleSub(
    B=&RigVMFunction_MathBoolToFloat(Value=RigVMVariableNode() {variable=isInFrontend}.Value).Result
    A=&RigVMFunction_MathDoubleSub(A=&RigVMFunction_MathDoubleSub(A=1, B=@n45.Value).Result, B=RigVMVariableNode() {variable=L thigh forward}.Value).Result
   )
   > &RigUnit_ModifyTransforms(
    ItemToModify=((Item=(Type=Bone,Name="dyn_R_back_skirt_1"),Transform=(Rotation=(X=0.204,Y=0.1299,Z=-0.2394,W=0.9403),Translation=(X=0,Y=0,Z=0),Scale3D=(X=1,Y=1,Z=1))))
    WeightMinimum=0
    Weight=@n46.Result
   )
   > &RigUnit_ModifyTransforms(
    ItemToModify=((Item=(Type=Bone,Name="dyn_R_side_skirt_1"),Transform=(Rotation=(X=0,Y=0,Z=-0.1736,W=0.9848),Translation=(X=0,Y=0,Z=0),Scale3D=(X=1,Y=1,Z=1))))
    WeightMinimum=0
    Weight=@n46.Result
   )
   n47 = &RigVMFunction_MathDoubleSub(
    B=&RigVMFunction_MathBoolToFloat(Value=RigVMVariableNode() {variable=isDBNO}.Value).Result
    A=RigVMVariableNode() {variable=spineTranslationAlpha Lower}.Value
   )
   > &RigUnit_ModifyTransforms(
    ItemToModify=((Item=(Type=Bone,Name="dyn_L_back_skirt_1"),Transform=(Rotation=(X=0,Y=0,Z=-0.0872,W=0.9962),Translation=(X=0,Y=0,Z=0),Scale3D=(X=1,Y=1,Z=1))))
    WeightMinimum=0
    Weight=@n47.Result
   )
   > &RigUnit_ModifyTransforms(
    ItemToModify=((Item=(Type=Bone,Name="dyn_R_back_skirt_1"),Transform=(Rotation=(X=0,Y=0,Z=-0.2588,W=0.9659),Translation=(X=0,Y=-2,Z=0),Scale3D=(X=1,Y=1,Z=1))))
    WeightMinimum=0
    Weight=@n47.Result
   )
   n48 = &RigVMFunction_MathDoubleMul(A=RigVMVariableNode() {variable=R thigh forward}.Value, B=RigVMVariableNode() {variable=R calf alpha}.Value)
   n49 = RigVMVariableNode() {variable=spineTranslationAlpha Lower}
   n50 = &RigVMFunction_MathBoolToFloat(Value=RigVMVariableNode() {variable=isDBNO}.Value)
   > &RigUnit_ModifyTransforms(
    ItemToModify="((Item=(Type=Bone,Name=\"dyn_R_back_skirt_1\"),Transform=(Rotation=(X=0,Y=0,Z=-0.1736,W=0.9848),Translation=(X=0,Y=-2,Z=0),Scale3D=(X=1,Y=1,Z=1))),(Item=(Type=Bone,Name=\"dyn_R_back_skirt_2\"),Transform=(Rotation=(X=0,Y=0,Z=0,W=1),Translation=(X=0,Y=0,Z=0),Sca ...(truncated)"
    WeightMinimum=0
    Weight=&RigVMFunction_MathDoubleMul(
     B=&RigVMFunction_MathDoubleMul(B=@n48.Result, A=@n49.Value).Result
     A=&RigVMFunction_MathDoubleSub(A=1, B=@n50.Result).Result
    ).Result
   )
   RigVMFunction_MathBoolToFloat = &RigVMFunction_MathBoolToFloat(Value=RigVMVariableNode() {variable=isKneeSliding}.Value)
   n51 = &RigVMFunction_AlphaInterp(InterpSpeedDecreasing=10, InterpSpeedIncreasing=3, Scale=1, bClampResult=False, Value=@RigVMFunction_MathBoolToFloat.Result)
   > &RigUnit_ModifyTransforms(
    ItemToModify="((Item=(Type=Bone,Name=\"dyn_L_back_skirt_1\"),Transform=(Rotation=(X=0,Y=0,Z=-0.2588,W=0.9659),Translation=(X=25,Y=2,Z=0),Scale3D=(X=1,Y=1,Z=1))),(Item=(Type=Bone,Name=\"dyn_L_back_skirt_2\"),Transform=(Rotation=(X=0,Y=0,Z=0.3827,W=0.9239),Translation=(X=-1,Y ...(truncated)"
    WeightMinimum=0
    Weight=@n51.Result
   )
   > &RigUnit_ModifyTransforms(
    ItemToModify="((Item=(Type=Bone,Name=\"dyn_R_back_skirt_1\"),Transform=(Rotation=(X=0.2596,Y=-0.0353,Z=-0.2929,W=0.9195),Translation=(X=-25,Y=0,Z=0),Scale3D=(X=1,Y=1,Z=1))),(Item=(Type=Bone,Name=\"dyn_R_back_skirt_2\"),Transform=(Rotation=(X=0,Y=0,Z=0.3827,W=0.9239),Transla ...(truncated)"
    WeightMinimum=0
    Weight=@n51.Result
   )
   n52 = &RigVMFunction_MathBoolToFloat(Value=RigVMVariableNode() {variable=isDBNO}.Value)
   > &RigUnit_ModifyTransforms(
    ItemToModify=((Item=(Type=Bone,Name="dyn_L_back_skirt_1"),Transform=(Rotation=(X=-0.0283,Y=0.2146,Z=0.1274,W=0.9679),Translation=(X=0,Y=0,Z=0),Scale3D=(X=1,Y=1,Z=1))))
    WeightMinimum=0
    Weight=@n52.Result
   )
   > &RigUnit_ModifyTransforms(
    ItemToModify=((Item=(Type=Bone,Name="dyn_R_back_skirt_1"),Transform=(Rotation=(X=0.0262,Y=0.0831,Z=-0.2996,W=0.9501),Translation=(X=0,Y=0,Z=0),Scale3D=(X=1,Y=1,Z=1))))
    WeightMinimum=0
    Weight=@n52.Result
   )
   > &RigUnit_ParentConstraint(
    AdvancedSettings=(InterpolationType=Average,RotationOrderForFilter=XZY)
    Child=(Type=Bone,Name="dyn_hipAccessory_base")
    Filter=(TranslationFilter=(bX=true,bY=true,bZ=true),RotationFilter=(bX=true,bY=false,bZ=true),ScaleFilter=(bX=True,bY=True,bZ=True))
    Parents=((Item=(Type=Bone,Name="thigh_r"),Weight=1))
    Weight=0.1
   )
   > &RigUnit_ParentConstraint(
    AdvancedSettings=(InterpolationType=Average,RotationOrderForFilter=XZY)
    Child=(Type=Bone,Name="dyn_hipAccessory_base")
    Filter=(TranslationFilter=(bX=true,bY=true,bZ=true),RotationFilter=(bX=false,bY=true,bZ=false),ScaleFilter=(bX=True,bY=True,bZ=True))
    Parents=((Item=(Type=Bone,Name="thigh_r"),Weight=1))
    Weight=0.3
   )
   n53 = &RigVMFunction_AlphaInterp(
    InterpSpeedDecreasing=20
    InterpSpeedIncreasing=20
    Scale=1
    bClampResult=False
    Value=RigVMVariableNode() {variable=R thigh VeryForward}.Value
   )
   > &RigUnit_ModifyTransforms(
    ItemToModify=((Item=(Type=Bone,Name="dyn_hipAccessory_base"),Transform=(Rotation=(X=0.1001,Y=-0.1751,Z=-0.3238,W=0.9244),Translation=(X=-1,Y=0,Z=0),Scale3D=(X=1,Y=1,Z=1))))
    WeightMinimum=0
    Weight=@n53.Result
   )
   &RigVMFunction_MathDoubleAdd(B=10, A=@n10.Value)
   &RigVMFunction_MathDoubleAdd(B=10, A=@n15.Value)
   n54 = RigVMVariableNode() {variable=L thighOut delta}
   &RigVMFunction_MathDoubleMax(
    B=0.5
    A=&RigVMFunction_MathDoubleSub(A=@n20.Result, B=&RigVMFunction_MathDoubleSub(A=1, B=@n54.Value).Result).Result
   )
   &RigUnit_ModifyTransforms(
    ExecutePin=()
    ItemToModify=((Item=(Type=Bone,Name="dyn_R_back_skirt_1"),Transform=(Rotation=(X=0,Y=0.3827,Z=0,W=0.9239),Translation=(X=0,Y=0,Z=0),Scale3D=(X=1,Y=1,Z=1))))
    Weight=1
    WeightMinimum=0
   )
   Multiply = &RigVMFunction_MathDoubleMul(B=0.6, A=RigVMVariableNode() {variable=L thighOut delta}.Value)
   &RigVMFunction_MathDoubleSub(A=@n12.Result, B=@Multiply.Result)
  ]}
]
hierarchy [
 root {loc=<0 0 0>, rot=<0 0 0 1>}
 pelvis {parent=root, loc=<0 -0.5301 98.0571>, rot=<0.0127 0.707 -0.0127 -0.707>}
 spine_01 {parent=pelvis, loc=<6.1208 -0.9328 0>, rot=<0 0 0.034 -0.9994>}
 spine_02 {parent=spine_01, loc=<6.5996 -0.2095 0>, rot=<0 0 0.0238 -0.9997>}
 spine_03 {parent=spine_02, loc=<6.4914 -0.1237 0>, rot=<0 0 0.0149 -0.9999>}
 spine_04 {parent=spine_03, loc=<6.9598 0.0794 0>, rot=<0 0 -0.0132 -0.9999>}
 spine_05 {parent=spine_04, loc=<6.9987 0 0>, rot=<0 0 -0.0015 -1>}
 attach_cape {parent=spine_05, loc=<9.3216 -8.4501 0>, rot=<0.0616 -0.7044 -0.0616 -0.7044>}
 neck_01 {parent=spine_05, loc=<11.3138 0.1529 0>, rot=<0 0 -0.2003 -0.9797>}
 neck_02 {parent=neck_01, loc=<4.5627 0 0>, rot=<0 0 0.0504 -0.9987>}
 head {parent=neck_02, loc=<5.3914 0 0>, rot=<0 0 0.115 -0.9934>}
 FACIAL_C_FacialRoot {parent=head, loc=<1.3331 0.2606 0.0003>, rot=<0.4736 -0.5251 -0.4736 -0.5251>}
 FACIAL_C_UnderChin {parent=FACIAL_C_FacialRoot, loc=<0.0006 8.6656 2.6043>, rot=<0.5133 0 0 -0.8582>}
 FACIAL_L_UnderChin {parent=FACIAL_C_FacialRoot, loc=<2.8512 7.3966 1.5082>, rot=<0.4162 -0.3535 0.0881 -0.8331>}
 FACIAL_R_UnderChin {parent=FACIAL_C_FacialRoot, loc=<-2.8522 7.3957 1.5061>, rot=<0.4162 0.3534 -0.0881 -0.8332>}
 FACIAL_C_Jaw {parent=FACIAL_C_FacialRoot, loc=<-0.0002 2.109 0.9348>, rot=<0.3472 0 0 -0.9378>}
 FACIAL_C_Jawline {parent=FACIAL_C_Jaw, loc=<0.0001 2.9347 7.1191>, rot=<0.2742 0 0 -0.9617>}
 FACIAL_C_Chin {parent=FACIAL_C_Jaw, loc=<0.0002 1.3816 7.4149>, rot=<-0.241 0 0 -0.9705>}
 FACIAL_L_Jawline {parent=FACIAL_C_FacialRoot, loc=<3.9808 5.7108 2.8927>, rot=<0.2287 0.1565 -0.0578 -0.9591>}
 FACIAL_R_Jawline {parent=FACIAL_C_FacialRoot, loc=<-3.9815 5.7117 2.8944>, rot=<0.2286 -0.1567 0.0578 -0.9591>}
 dyn_front_strand {parent=head, loc=<6.2794 9.0164 0.1271>, rot=<-0.052 -0.9944 0.091 -0.0107>}
 dyn_L_hair_main_1 {parent=head, loc=<3.011 -5.4573 -3.3315>, rot=<0.9941 0.0885 0.0627 0.0051>}
 dyn_L_hair_main_2 {parent=dyn_L_hair_main_1, loc=<-8.1808 0 0>, rot=<-0.0119 0.0378 0.1345 -0.9901>}
 dyn_L_hair_main_3 {parent=dyn_L_hair_main_2, loc=<-6.6371 0 0>, rot=<-0.0078 0.022 -0.0976 -0.995>}
 dyn_L_hair_main_4 {parent=dyn_L_hair_main_3, loc=<-9.2744 0 0>, rot=<0 0 -0.1064 -0.9943>}
 dyn_L_hair_main_5 {parent=dyn_L_hair_main_4, loc=<-10.0047 -0.0001 0>, rot=<0 0 -0.112 -0.9937>}
 dyn_R_hair_main_1 {parent=head, loc=<3.011 -5.4573 3.3315>, rot=<0.0885 -0.9941 -0.0051 0.0627>}
 dyn_R_hair_main_2 {parent=dyn_R_hair_main_1, loc=<8.1808 0 0>, rot=<-0.0119 0.0378 0.1345 -0.9901>}
 dyn_R_hair_main_3 {parent=dyn_R_hair_main_2, loc=<6.637 0 0>, rot=<-0.0078 0.022 -0.0976 -0.995>}
 dyn_R_hair_main_4 {parent=dyn_R_hair_main_3, loc=<9.2744 0 0>, rot=<0 0 -0.1064 -0.9943>}
 dyn_R_hair_main_5 {parent=dyn_R_hair_main_4, loc=<10.0047 0 0>, rot=<0 0 -0.112 -0.9937>}
 dyn_L_sideHair_1 {parent=head, loc=<3.1178 3.867 -6.4363>, rot=<0.0385 -0.9988 0.0009 -0.031>}
 dyn_L_sideHair_2 {parent=dyn_L_sideHair_1, loc=<5.7059 0 0>, rot=<0.0064 -0.0709 0.0207 -0.9973>}
 dyn_L_sideHair_strand_A {parent=dyn_L_sideHair_2, loc=<3.5677 0 0>, rot=<0.0261 -0.119 -0.1268 -0.9844>}
 dyn_L_sideHair_strand_B {parent=dyn_L_sideHair_2, loc=<2.796 1.416 -0.2399>, rot=<0.0261 -0.119 -0.1268 -0.9844>}
 dyn_R_sideHair_1 {parent=head, loc=<3.1645 3.2637 6.4337>, rot=<-0.0831 -0.9961 0.0028 0.0309>}
 dyn_R_sideHair_2 {parent=dyn_R_sideHair_1, loc=<3.9781 0 0>, rot=<0.0003 0.0198 0.1586 -0.9871>}
 dyn_R_sideHair_strand_A {parent=dyn_R_sideHair_2, loc=<5.3575 0 0>, rot=<0.0125 0.1285 -0.0963 -0.9869>}
 dyn_R_sideHair_strand_B {parent=dyn_R_sideHair_2, loc=<3.9779 1.453 1.4003>, rot=<0.0166 0.2095 -0.0774 -0.9746>}
 dyn_L_side_strand_1 {parent=head, loc=<1.7133 0.755 -7.2361>, rot=<0.3113 -0.9029 0.0943 -0.2811>}
 dyn_L_side_strand_2 {parent=dyn_L_side_strand_1, loc=<5.2168 0 0>, rot=<0 -0.1295 0 -0.9916>}
 dyn_L_sideHair_long_1 {parent=head, loc=<3.1178 3.867 -6.4363>, rot=<0.2467 -0.9573 0.0364 -0.1463>}
 dyn_L_sideHair_long_2 {parent=dyn_L_sideHair_long_1, loc=<7.6479 0 0>, rot=<-0.036 0.0481 0.1456 -0.9875>}
 dyn_L_sideHair_long_3 {parent=dyn_L_sideHair_long_2, loc=<5.4548 0 0>, rot=<0.2427 -0.173 0.1545 -0.9419>}
 dyn_L_sideHair_long_4 {parent=dyn_L_sideHair_long_3, loc=<6.8607 0 0>, rot=<0.2229 -0.1801 -0.2118 -0.9344>}
 dyn_L_sideHair_long_5 {parent=dyn_L_sideHair_long_4, loc=<7.18 0 0>, rot=<-0.0801 0.1723 -0.2267 -0.9553>}
 dyn_L_sideHair_long_6 {parent=dyn_L_sideHair_long_5, loc=<5.2323 0 0>, rot=<0 0.1408 0 -0.99>}
 dyn_R_sideHair_long_1 {parent=head, loc=<3.1645 3.2637 6.4337>, rot=<0.2485 -0.9616 -0.0282 0.113>}
 dyn_R_sideHair_long_2 {parent=dyn_R_sideHair_long_1, loc=<4.092 0 0>, rot=<0.0195 -0.0404 -0.0485 -0.9978>}
 dyn_R_sideHair_long_3 {parent=dyn_R_sideHair_long_2, loc=<5.8442 0 0>, rot=<-0.0438 0.0428 0.3855 -0.9207>}
 dyn_R_sideHair_long_A {parent=dyn_R_sideHair_long_3, loc=<3.7273 0 0>, rot=<0 0 -0.1381 -0.9904>}
 dyn_R_sideHair_long_4 {parent=dyn_R_sideHair_long_3, loc=<4.966 3.212 -1.0589>, rot=<-0.1598 0.0831 -0.1003 -0.9785>}
 dyn_R_sideHair_long_5 {parent=dyn_R_sideHair_long_4, loc=<8.8285 0 0>, rot=<-0.0784 0.0955 -0.2948 -0.9475>}
 dyn_R_sideHair_long_6 {parent=dyn_R_sideHair_long_5, loc=<7.5052 0 0>, rot=<0.0281 -0.1816 -0.2265 -0.9565>}
 dyn_R_sideHair_long_7 {parent=dyn_R_sideHair_long_6, loc=<6.8372 0 0>, rot=<0 -0.1223 0 -0.9925>}
 attach_backpack {parent=spine_05, loc=<0 -8.35 0>, rot=<-0.0093 -0.707 0.0093 -0.707>}
 clavicle_l {parent=spine_05, loc=<5.4831 -0.1053 -3.2946>, rot=<0.0465 -0.7289 -0.009 -0.6829>}
 upperarm_l {parent=clavicle_l, loc=<9.8658 -1.2136 -0.4105>, rot=<0.0164 -0.3245 -0.0269 -0.9454>}
 lowerarm_l {parent=upperarm_l, loc=<26.465 -0.8345 0.259>, rot=<0.0036 0.0161 -0.2298 -0.9731>}
 hand_l {parent=lowerarm_l, loc=<22.7355 -0.3203 -0.6105>, rot=<-0.6784 -0.0223 0.0735 -0.7307>}
 thumb_01_l {parent=hand_l, loc=<1.3118 -0.3191 -2.459>, rot=<0.6546 -0.4042 0.0664 -0.6354>}
 thumb_02_l {parent=thumb_01_l, loc=<5.9336 -0.0448 0.0401>, rot=<-0.0037 0.0213 0.298 -0.9543>}
 thumb_03_l {parent=thumb_02_l, loc=<2.2223 0.3682 -0.0907>, rot=<0.0118 -0.013 0.0066 -0.9998>}
 index_metacarpal_l {parent=hand_l, loc=<1.6535 0.0183 -2.96>, rot=<0.0158 -0.0124 -0.0173 -0.9996>}
 index_01_l {parent=index_metacarpal_l, loc=<8.3008 0.0357 -0.0524>, rot=<-0.0044 0.0485 0.0924 -0.9945>}
 index_02_l {parent=index_01_l, loc=<3.3587 -0.3004 0.0452>, rot=<-0.002 0.0027 0.3225 -0.9466>}
 index_03_l {parent=index_02_l, loc=<2.5578 -0.123 -0.0252>, rot=<-0.002 0.0067 0.2558 -0.9667>}
 middle_metacarpal_l {parent=hand_l, loc=<2.1856 -0.1725 -0.5984>, rot=<0.006 0.0168 -0.0329 -0.9993>}
 middle_01_l {parent=middle_metacarpal_l, loc=<7.3275 -0.2447 0.0129>, rot=<-0.0208 0.0042 0.0858 -0.9961>}
 middle_02_l {parent=middle_01_l, loc=<4.0123 0.1401 0.1175>, rot=<-0.0218 0.0258 0.3512 -0.9357>}
 middle_03_l {parent=middle_02_l, loc=<2.6077 0.0534 0.1373>, rot=<0.0082 0.0188 0.1708 -0.9851>}
 ring_metacarpal_l {parent=hand_l, loc=<2.2074 -0.3625 1.3638>, rot=<-0.0284 0.0722 -0.0458 -0.9959>}
 ring_01_l {parent=ring_metacarpal_l, loc=<7.0295 -0.6197 -0.1446>, rot=<-0.0489 -0.0252 0.1892 -0.9804>}
 ring_02_l {parent=ring_01_l, loc=<3.6174 0.4647 -0.1027>, rot=<0 0 0.2806 -0.9598>}
 ring_03_l {parent=ring_02_l, loc=<2.4152 0.114 0.0199>, rot=<0 0 0.1189 -0.9929>}
 pinky_metacarpal_l {parent=hand_l, loc=<1.7616 -0.9753 3.3814>, rot=<-0.0397 0.0814 -0.069 -0.9935>}
 pinky_01_l {parent=pinky_metacarpal_l, loc=<7.1709 -0.4831 -0.1094>, rot=<-0.0654 -0.0035 0.2242 -0.9723>}
 pinky_02_l {parent=pinky_01_l, loc=<2.8274 0.0893 -0.1043>, rot=<0.0037 0.0011 0.2854 -0.9584>}
 pinky_03_l {parent=pinky_02_l, loc=<2.2046 0.0307 -0.0078>, rot=<-0.0002 0.0098 0.1445 -0.9895>}
 weapon_l {parent=hand_l, loc=<-0.1974 -3.4111 0.3811>, rot=<0 0 0 1>}
 hand_attach_l {parent=hand_l, loc=<8.5 -3 0>, rot=<0.7071 0.7071 0 0>}
 lowerarm_twist_02_l {parent=lowerarm_l, loc=<10.6707 0 0>, rot=<0 0 0 1>}
 lowerarm_twist_01_l {parent=lowerarm_l, loc=<18.316 0 0>, rot=<0 0 0 1>}
 upperarm_twist_01_l {parent=upperarm_l, loc=<7.2135 -0.1134 0>, rot=<0 0 0 1>}
 upperarm_twist_02_l {parent=upperarm_l, loc=<19.7998 -0.1134 0>, rot=<0 0 0 1>}
 clavicle_r {parent=spine_05, loc=<5.4831 -0.1053 3.2946>, rot=<0.7289 0.0465 -0.6829 0.009>}
 upperarm_r {parent=clavicle_r, loc=<-9.8658 1.2136 0.4105>, rot=<0.0164 -0.3245 -0.0269 -0.9454>}
 lowerarm_r {parent=upperarm_r, loc=<-26.465 0.8345 -0.259>, rot=<0.0036 0.0161 -0.2298 -0.9731>}
 hand_r {parent=lowerarm_r, loc=<-22.7355 0.3203 0.6106>, rot=<-0.6784 -0.0223 0.0735 -0.7307>}
 thumb_01_r {parent=hand_r, loc=<-1.3118 0.3191 2.459>, rot=<0.6546 -0.4042 0.0664 -0.6354>}
 thumb_02_r {parent=thumb_01_r, loc=<-5.9336 0.0448 -0.0401>, rot=<-0.0037 0.0213 0.298 -0.9543>}
 thumb_03_r {parent=thumb_02_r, loc=<-2.2223 -0.3682 0.0907>, rot=<0.0118 -0.013 0.0066 -0.9998>}
 index_metacarpal_r {parent=hand_r, loc=<-1.6535 -0.0183 2.96>, rot=<0.0158 -0.0124 -0.0173 -0.9996>}
 index_01_r {parent=index_metacarpal_r, loc=<-8.3008 -0.0357 0.0524>, rot=<-0.0044 0.0485 0.0924 -0.9945>}
 index_02_r {parent=index_01_r, loc=<-3.3587 0.3004 -0.0452>, rot=<-0.002 0.0027 0.3225 -0.9466>}
 index_03_r {parent=index_02_r, loc=<-2.5578 0.123 0.0252>, rot=<-0.002 0.0067 0.2558 -0.9667>}
 middle_metacarpal_r {parent=hand_r, loc=<-2.1856 0.1725 0.5984>, rot=<0.006 0.0168 -0.0329 -0.9993>}
 middle_01_r {parent=middle_metacarpal_r, loc=<-7.3275 0.2447 -0.0129>, rot=<-0.0208 0.0042 0.0858 -0.9961>}
 middle_02_r {parent=middle_01_r, loc=<-4.0123 -0.1401 -0.1175>, rot=<-0.0218 0.0258 0.3512 -0.9357>}
 middle_03_r {parent=middle_02_r, loc=<-2.6077 -0.0534 -0.1373>, rot=<0.0082 0.0188 0.1708 -0.9851>}
 ring_metacarpal_r {parent=hand_r, loc=<-2.2074 0.3625 -1.3638>, rot=<-0.0284 0.0722 -0.0458 -0.9959>}
 ring_01_r {parent=ring_metacarpal_r, loc=<-7.0295 0.6197 0.1446>, rot=<-0.0489 -0.0252 0.1892 -0.9804>}
 ring_02_r {parent=ring_01_r, loc=<-3.6174 -0.4647 0.1027>, rot=<0 0 0.2806 -0.9598>}
 ring_03_r {parent=ring_02_r, loc=<-2.4152 -0.114 -0.0199>, rot=<0 0 0.1189 -0.9929>}
 pinky_metacarpal_r {parent=hand_r, loc=<-1.7616 0.9753 -3.3814>, rot=<-0.0397 0.0814 -0.069 -0.9935>}
 pinky_01_r {parent=pinky_metacarpal_r, loc=<-7.1709 0.4831 0.1094>, rot=<-0.0654 -0.0035 0.2242 -0.9723>}
 pinky_02_r {parent=pinky_01_r, loc=<-2.8274 -0.0893 0.1043>, rot=<0.0037 0.0011 0.2854 -0.9584>}
 pinky_03_r {parent=pinky_02_r, loc=<-2.2046 -0.0307 0.0078>, rot=<-0.0002 0.0098 0.1445 -0.9895>}
 weapon_r {parent=hand_r, loc=<0.1974 3.4111 -0.3811>, rot=<0 0 0 1>}
 hand_attach_r {parent=hand_r, loc=<-8.5001 2.9999 0>, rot=<0 0 -0.7071 -0.7071>}
 lowerarm_twist_02_r {parent=lowerarm_r, loc=<-10.6707 0 0>, rot=<0 0 0 -1>}
 lowerarm_twist_01_r {parent=lowerarm_r, loc=<-18.316 0 0>, rot=<0 0 0 -1>}
 upperarm_twist_01_r {parent=upperarm_r, loc=<-7.2135 0.1134 0>, rot=<0 0 0 -1>}
 upperarm_twist_02_r {parent=upperarm_r, loc=<-19.7998 0.1134 0>, rot=<0 0 0 -1>}
 thigh_l {parent=pelvis, loc=<-1.0855 -0.2109 -8.7053>, rot=<0.0009 0.0355 0.0155 -0.9993>}
 calf_l {parent=thigh_l, loc=<-42.5284 -0.0019 -0.7297>, rot=<0.0006 -0.0161 -0.0391 -0.9991>}
 foot_l {parent=calf_l, loc=<-46.7401 -2.5014 -0.592>, rot=<0.0427 -0.0238 0.0353 -0.9982>}
 ball_l {parent=foot_l, loc=<-3.5658 10.8066 -0.1544>, rot=<0.0001 -0.0001 -0.7063 -0.7079>}
 calf_twist_02_l {parent=calf_l, loc=<-19.1546 -1.7943 -0.2638>, rot=<0.0237 0.0095 -0.0035 -0.9997>}
 calf_twist_01_l {parent=calf_l, loc=<-35.8483 -1.9197 -0.5794>, rot=<0.0237 0.0095 -0.0035 -0.9997>}
 thigh_twist_01_l {parent=thigh_l, loc=<-27.952 0 0>, rot=<0.0013 0 0 -1>}
 dyn_L_side_skirt_1 {parent=thigh_l, loc=<-11.1615 -1.6875 -6.9056>, rot=<0.8084 -0.0012 -0.0709 -0.5843>}
 dyn_L_side_skirt_2 {parent=dyn_L_side_skirt_1, loc=<-17.143 0 0>, rot=<0 0 0 1>}
 thigh_r {parent=pelvis, loc=<-1.0855 -0.2109 8.7053>, rot=<0.0355 -0.0009 0.9993 0.0155>}
 calf_r {parent=thigh_r, loc=<42.5284 0.0019 0.7297>, rot=<0.0006 -0.0161 -0.0391 -0.9991>}
 foot_r {parent=calf_r, loc=<46.7401 2.5014 0.592>, rot=<0.0427 -0.0238 0.0353 -0.9982>}
 ball_r {parent=foot_r, loc=<3.5658 -10.8066 0.1544>, rot=<0.0001 -0.0001 -0.7063 -0.7079>}
 calf_twist_02_r {parent=calf_r, loc=<19.1546 1.7943 0.2638>, rot=<0.0237 0.0095 -0.0035 -0.9997>}
 calf_twist_01_r {parent=calf_r, loc=<35.8483 1.9197 0.5794>, rot=<0.0237 0.0095 -0.0035 -0.9997>}
 thigh_twist_01_r {parent=thigh_r, loc=<27.952 0 0>, rot=<0.0013 0 0 -1>}
 dyn_R_side_skirt_1 {parent=thigh_r, loc=<18.4992 -0.6829 6.9275>, rot=<0.6051 -0.0203 -0.0343 -0.7951>}
 dyn_R_side_skirt_2 {parent=dyn_R_side_skirt_1, loc=<12.171 0 0>, rot=<0 0 0 1>}
 dyn_R_side_skirt_3 {parent=dyn_R_side_skirt_2, loc=<12.7123 0 0>, rot=<0 0 0 1>}
 dyn_front_skirt_1 {parent=pelvis, loc=<-18.7683 7.0836 4.8451>, rot=<-0.0015 0 1 0.0005>}
 dyn_front_skirt_2 {parent=dyn_front_skirt_1, loc=<12.185 0 0>, rot=<0 0 0 1>}
 dyn_front_skirt_3 {parent=dyn_front_skirt_2, loc=<12.7149 0 0>, rot=<0 0 0 1>}
 dyn_R_back_skirt_1 {parent=pelvis, loc=<-2.2425 -9.494 9.6488>, rot=<-0.0007 -0.9921 0.1257 -0.0014>}
 dyn_R_back_skirt_2 {parent=dyn_R_back_skirt_1, loc=<29.3498 1.4484 -0.4268>, rot=<0 0 0 1>}
 dyn_R_back_skirt_3 {parent=dyn_R_back_skirt_2, loc=<12.7227 0 0>, rot=<0 0 0 1>}
 dyn_L_skirtDriver_1 {parent=pelvis, loc=<2.3414 6.1838 -3.5721>, rot=<-0.0015 0 1 0.0005>}
 dyn_L_skirtDriver_2 {parent=dyn_L_skirtDriver_1, loc=<7.414 0 0>, rot=<0 0 0 1>}
 dyn_L_back_skirt_1 {parent=pelvis, loc=<-2.2425 -9.494 -9.6488>, rot=<0.9921 -0.0007 -0.0014 -0.1257>}
 dyn_L_back_skirt_2 {parent=dyn_L_back_skirt_1, loc=<-29.3499 -1.4484 0.4268>, rot=<0 0 0 1>}
 dyn_L_back_skirt_3 {parent=dyn_L_back_skirt_2, loc=<-12.7226 0 0>, rot=<0 0 0 1>}
 dyn_zipper {parent=pelvis, loc=<6.602 6.435 7.0657>, rot=<-0.3836 -0.6679 -0.2461 -0.5885>}
 dyn_hipAccessory_base {parent=pelvis, loc=<6.913 1.2307 11.3406>, rot=<-0.5414 0.0811 0.8152 -0.189>}
 dyn_hipAccessory_dangleElement {parent=dyn_hipAccessory_base, loc=<-2.3968 -1.9964 10.0931>, rot=<-0.2643 -0.7593 0.5852 -0.105>}
 attach {parent=root, loc=<0 0 0>, rot=<0 0 0 1>}
 ik_hand_root {parent=root, loc=<0 0 0>, rot=<0 0 0 1>}
 ik_hand_gun {parent=ik_hand_root, loc=<-48.1743 3.3545 104.4381>, rot=<0.6276 0.1309 0.4183 -0.6434>}
 ik_hand_r {parent=ik_hand_gun, loc=<0 0 0>, rot=<0 0 0 1>}
 ik_hand_l {parent=ik_hand_gun, loc=<59.3271 67.6997 34.3524>, rot=<-0.6158 -0.7027 -0.3565 0>}
 ik_foot_root {parent=root, loc=<0 0 0>, rot=<0 0 0 1>}
 ik_foot_l {parent=ik_foot_root, loc=<14.8262 -7.3545 8.296>, rot=<0.0357 0.7045 0.0284 -0.7082>}
 ik_foot_r {parent=ik_foot_root, loc=<-14.8262 -7.3545 8.296>, rot=<0.7082 0.0284 -0.7045 0.0357>}
 attach_fp {parent=root, loc=<0 0 0>, rot=<0 0 0 1>}
 camera_root {parent=root, loc=<0 0 0>, rot=<0 0 0 1>}
 camera {parent=camera_root, loc=<0 0 0>, rot=<0 0 0 1>}
 attach_emote_01 {parent=root, loc=<0 0 0>, rot=<0 0 0 1>}
 attach_emote_02 {parent=root, loc=<0 0 0>, rot=<0 0 0 1>}
 attach_emote_03 {parent=root, loc=<0 0 0>, rot=<0 0 0 1>}
 warp_point_01 {parent=root, loc=<0 0 0>, rot=<0 0 0 1>}
]
variables [
 L calf alpha {type=double, default=0}
 L thigh VeryForward {type=double, default=0}
 L thigh forward {type=float, default=0}
 L thighOut delta {type=double, default=0}
 L_thigh_Z_value {type=double, default=0}
 R calf alpha {type=double, default=0}
 R thigh VeryForward {type=double, default=0}
 R thigh forward {type=float, default=0}
 R thighOut delta {type=double, default=0}
 R_thigh_Z_value {type=double, default=0}
 isDBNO {type=bool, default=False}
 isInFrontend {type=bool, default=False}
 isKneeSliding {type=bool, default=False}
 isSkydiving {type=bool, default=False}
 isSwimming {type=bool, default=False}
 spineTranslationAlpha Higher {type=double, default=0}
 spineTranslationAlpha Lower {type=double, default=0}
]
