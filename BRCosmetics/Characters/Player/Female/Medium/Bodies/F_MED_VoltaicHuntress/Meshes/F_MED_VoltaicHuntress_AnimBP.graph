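# Animation Blueprint for the Voltaic Huntress female body cosmetic, adding secondary motion to a pose copied from the attached parent mesh.
# A Control Rig receives movement and frontend state flags, then a weighted constraint positions the zipper bone.
# Chained AnimDynamics nodes simulate zipper, hip dangle and skirt bones, followed by a small left-skirt offset.
# The EventGraph updates GravityOverride via a deprecated function and reads a frontend idle bypass curve.
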AnimBlueprint /BRCosmetics/Characters/Player/Female/Medium/Bodies/F_MED_VoltaicHuntress/Meshes/F_MED_VoltaicHuntress_AnimBP
target_skeleton=&/F_MED_VoltaicHuntress_Skeleton
graphs [
 AnimGraph {nodes [
   n1 = Control Rig!(
    Source=Copy Pose From Mesh() {bUseAttachedParent=true}
    Alpha
    isKneeSliding:bool=Get bIsKneeSliding()
    isSkydiving:bool=Get bIsSkyDiving()
    isSwimming:bool=Get bIsSurfaceSwimming()
    isDBNO:bool=Get bIsDBNO()
    isInFrontend:bool=Get bIsInFrontEnd()
   ) -> (Pose) {
    alphaBoolBlend {blendOption=Linear}
    controlRigAssetReference {blueprintRigClass=&/F_MED_VoltaicHuntress_CtrlRig.F_MED_VoltaicHuntress_CtrlRig_C}
   }
   n2 = "Constraint - dyn_zipper (PPPP)"!(
    ComponentPose=Local To Component(LocalPose=@n1)
    Alpha=1
    ConstraintWeights_0=0.8654
    ConstraintWeights_1:float=0.072
    ConstraintWeights_2:float=0.0501
    ConstraintWeights_3:float=0.0126
   ) -> (Pose) {
    alphaBoolBlend {blendOption=Linear}
    boneToModify {boneName=dyn_zipper}
    constraintSetup [{targetBone {boneName=pelvis}}, {targetBone {boneName=thigh_r}}, {targetBone {boneName=spine_01}}, {targetBone {boneName=spine_02}}]
    constraintWeights [1,1,1,1]
   }
   n3 = Anim Dynamics - Bone: dyn_zipper(ComponentPose=@n2, Alpha=1) {
    angularDampingOverride=0.99
    angularSpringConstant=3500
    bAngularSpring=true
    bOverrideAngularDamping=true
    bUseGravityOverride=true
    gravityScale=0.35
    simulationSpace=RootRelative
    alphaBoolBlend {blendOption=Linear}
    boundBone {boneName=dyn_zipper}
    componentLinearAccScale=<1 1 1>
    componentLinearVelScale=<1 1 1>
    gravityOverride {z=-980}
    physicsBodyDefinitions [{
      boundBone {boneName=dyn_zipper}
      boxExtents=<12 12 12>
      constraintSetup {angularLimitsMax=<1 10 1>, angularLimitsMin=<-10 -10 -5>, angularTarget=<1 0 0>}
      localJointOffset=<0 0 -2>
     }]
   }
   n4 = Anim Dynamics - Bone: dyn_hipAccessory_dangleElement(ComponentPose=@n3, Alpha=1, ExternalForce="0,0,0") {
    angularDampingOverride=0.999
    angularSpringConstant=50
    bAngularSpring=true
    bOverrideAngularDamping=true
    bUseGravityOverride=true
    gravityScale=0.35
    simulationSpace=RootRelative
    alphaBoolBlend {blendOption=Linear}
    boundBone {boneName=dyn_hipAccessory_dangleElement}
    componentLinearAccScale=<1 1 1>
    componentLinearVelScale=<1 1 1>
    gravityOverride {z=-800}
    physicsBodyDefinitions [{
      boundBone {boneName=dyn_hipAccessory_dangleElement}
      boxExtents=<10 10 12>
      constraintSetup {angularLimitsMax=<30 3 80>, angularLimitsMin=<-30 -8 0>, angularTarget=<1 0 0>}
      localJointOffset=<-1 -2 0>
     }]
   }
   n5 = Anim Dynamics - Bone: dyn_L_skirtDriver_1!(ComponentPose=@n4, Alpha=1, GravityOverride:Vector="0,0,-980", ExternalForce="0,0,0") -> (Pose) {
    angularDampingOverride=1
    angularSpringConstant=5000
    bAngularSpring=true
    bOverrideAngularDamping=true
    bUseGravityOverride=true
    gravityScale=0.35
    simulationSpace=RootRelative
    alphaBoolBlend {blendOption=Linear}
    boundBone {boneName=dyn_L_skirtDriver_1}
    componentLinearAccScale=<1 1 1>
    componentLinearVelScale=<1 1 1>
    gravityOverride {z=-980}
    physicsBodyDefinitions [{
      boundBone {boneName=dyn_L_skirtDriver_1}
      boxExtents=<20 30 30>
      constraintSetup {angularLimitsMax=<0 0 0>, angularLimitsMin=<0 -15 -15>, angularTarget=<1 0 0>}
      localJointOffset=<3 0 0>
     }]
   }
   n6 = Anim Dynamics - Bone: dyn_L_skirtDriver_2!(ComponentPose=@n5, Alpha=1, GravityOverride:Vector="0,0,-980", ExternalForce="0,0,0") -> (Pose) {
    angularDampingOverride=1
    angularSpringConstant=5000
    bAngularSpring=true
    bOverrideAngularDamping=true
    bUseGravityOverride=true
    gravityScale=0.35
    simulationSpace=RootRelative
    alphaBoolBlend {blendOption=Linear}
    boundBone {boneName=dyn_L_skirtDriver_2}
    componentLinearAccScale=<1 1 1>
    componentLinearVelScale=<1 1 1>
    gravityOverride {z=-980}
    physicsBodyDefinitions [{
      boundBone {boneName=dyn_L_skirtDriver_2}
      boxExtents=<20 30 30>
      constraintSetup {angularLimitsMax=<0 0 0>, angularLimitsMin=<0 -15 -15>, angularTarget=<1 0 0>}
      localJointOffset=<3 0 0>
     }]
   }
   n7 = Anim Dynamics - Bone: dyn_L_side_skirt_1!(ComponentPose=@n6, Alpha=1, GravityOverride:Vector="0,0,-980", ExternalForce="0,0,0") -> (Pose) {
    angularDampingOverride=1
    angularSpringConstant=5000
    bAngularSpring=true
    bOverrideAngularDamping=true
    bUseGravityOverride=true
    gravityScale=0.35
    simulationSpace=RootRelative
    alphaBoolBlend {blendOption=Linear}
    boundBone {boneName=dyn_L_side_skirt_1}
    componentLinearAccScale=<1 1 1>
    componentLinearVelScale=<1 1 1>
    gravityOverride {z=-980}
    physicsBodyDefinitions [{
      boundBone {boneName=dyn_L_side_skirt_1}
      boxExtents=<25 25 25>
      constraintSetup {angularLimitsMax=<0 0 0>, angularLimitsMin=<0 -10 -15>, angularTarget=<1 0 0>}
      localJointOffset=<-4 0 0>
     }]
   }
   n8 = Anim Dynamics - Bone: dyn_L_side_skirt_2!(ComponentPose=@n7, Alpha=1, GravityOverride:Vector="0,0,-980", ExternalForce="0,0,0") -> (Pose) {
    angularDampingOverride=1
    angularSpringConstant=5000
    bAngularSpring=true
    bOverrideAngularDamping=true
    bUseGravityOverride=true
    gravityScale=0.35
    simulationSpace=RootRelative
    alphaBoolBlend {blendOption=Linear}
    boundBone {boneName=dyn_L_side_skirt_2}
    componentLinearAccScale=<1 1 1>
    componentLinearVelScale=<1 1 1>
    gravityOverride {z=-980}
    physicsBodyDefinitions [{
      boundBone {boneName=dyn_L_side_skirt_2}
      boxExtents=<25 25 25>
      constraintSetup {angularLimitsMax=<0 0 0>, angularLimitsMin=<0 -10 -15>, angularTarget=<1 0 0>}
      localJointOffset=<-4 0 0>
     }]
   }
   n9 = Anim Dynamics - Bone: dyn_L_back_skirt_1!(ComponentPose=@n8, Alpha=1, GravityOverride:Vector="0,0,-980", ExternalForce="0,0,0") -> (Pose) {
    angularDampingOverride=1
    angularSpringConstant=5000
    bAngularSpring=true
    bOverrideAngularDamping=true
    bUseGravityOverride=true
    gravityScale=0.35
    simulationSpace=RootRelative
    alphaBoolBlend {blendOption=Linear}
    boundBone {boneName=dyn_L_back_skirt_1}
    componentLinearAccScale=<1 1 1>
    componentLinearVelScale=<1 1 1>
    gravityOverride {z=-980}
    physicsBodyDefinitions [{
      boundBone {boneName=dyn_L_back_skirt_1}
      boxExtents=<25 25 25>
      constraintSetup {angularLimitsMax=<0 0 0>, angularLimitsMin=<0 -10 -20>, angularTarget=<1 0 0>}
      localJointOffset=<-4 0 0>
     }]
   }
   n10 = Anim Dynamics - Bone: dyn_L_back_skirt_2!(ComponentPose=@n9, Alpha=1, GravityOverride:Vector="0,0,-980", ExternalForce="0,0,0") -> (Pose) {
    angularDampingOverride=1
    angularSpringConstant=5000
    bAngularSpring=true
    bOverrideAngularDamping=true
    bUseGravityOverride=true
    gravityScale=0.35
    simulationSpace=RootRelative
    alphaBoolBlend {blendOption=Linear}
    boundBone {boneName=dyn_L_back_skirt_2}
    componentLinearAccScale=<1 1 1>
    componentLinearVelScale=<1 1 1>
    gravityOverride {z=-980}
    physicsBodyDefinitions [{
      boundBone {boneName=dyn_L_back_skirt_2}
      boxExtents=<25 25 25>
      constraintSetup {angularLimitsMax=<0 0 0>, angularLimitsMin=<0 -10 -20>, angularTarget=<1 0 0>}
      localJointOffset=<-4 0 0>
     }]
   }
   n11 = Anim Dynamics - Bone: dyn_L_back_skirt_3!(ComponentPose=@n10, Alpha=1, GravityOverride:Vector="0,0,-980", ExternalForce="0,0,0") -> (Pose) {
    angularDampingOverride=1
    angularSpringConstant=5000
    bAngularSpring=true
    bOverrideAngularDamping=true
    bUseGravityOverride=true
    gravityScale=0.35
    simulationSpace=RootRelative
    alphaBoolBlend {blendOption=Linear}
    boundBone {boneName=dyn_L_back_skirt_3}
    componentLinearAccScale=<1 1 1>
    componentLinearVelScale=<1 1 1>
    gravityOverride {z=-980}
    physicsBodyDefinitions [{
      boundBone {boneName=dyn_L_back_skirt_3}
      boxExtents=<25 25 25>
      constraintSetup {angularLimitsMax=<0 0 0>, angularLimitsMin=<0 -10 -20>, angularTarget=<1 0 0>}
      localJointOffset=<-4 0 0>
     }]
   }
   n12 = Anim Dynamics - Bone: dyn_front_skirt_1!(ComponentPose=@n11, Alpha=1, GravityOverride:Vector="0,0,-980", ExternalForce="0,0,0") -> (Pose) {
    angularDampingOverride=1
    angularSpringConstant=5000
    bAngularSpring=true
    bOverrideAngularDamping=true
    bUseGravityOverride=true
    gravityScale=0.35
    simulationSpace=RootRelative
    alphaBoolBlend {blendOption=Linear}
    boundBone {boneName=dyn_front_skirt_1}
    componentLinearAccScale=<1 1 1>
    componentLinearVelScale=<1 1 1>
    gravityOverride {z=-980}
    physicsBodyDefinitions [{
      boundBone {boneName=dyn_front_skirt_1}
      boxExtents=<30 30 30>
      constraintSetup {angularLimitsMax=<0 0 0>, angularLimitsMin=<0 -10 -15>, angularTarget=<1 0 0>}
      localJointOffset=<4 0 0>
     }]
   }
   n13 = Anim Dynamics - Bone: dyn_front_skirt_2!(ComponentPose=@n12, Alpha=1, GravityOverride:Vector="0,0,-980", ExternalForce="0,0,0") -> (Pose) {
    angularDampingOverride=1
    angularSpringConstant=5000
    bAngularSpring=true
    bOverrideAngularDamping=true
    bUseGravityOverride=true
    gravityScale=0.35
    simulationSpace=RootRelative
    alphaBoolBlend {blendOption=Linear}
    boundBone {boneName=dyn_front_skirt_2}
    componentLinearAccScale=<1 1 1>
    componentLinearVelScale=<1 1 1>
    gravityOverride {z=-980}
    physicsBodyDefinitions [{
      boundBone {boneName=dyn_front_skirt_2}
      boxExtents=<30 30 30>
      constraintSetup {angularLimitsMax=<0 0 0>, angularLimitsMin=<0 -10 -15>, angularTarget=<1 0 0>}
      localJointOffset=<4 0 0>
     }]
   }
   n14 = Anim Dynamics - Bone: dyn_front_skirt_3!(ComponentPose=@n13, Alpha=1, GravityOverride:Vector="0,0,-980", ExternalForce="0,0,0") -> (Pose) {
    angularDampingOverride=1
    angularSpringConstant=5000
    bAngularSpring=true
    bOverrideAngularDamping=true
    bUseGravityOverride=true
    gravityScale=0.35
    simulationSpace=RootRelative
    alphaBoolBlend {blendOption=Linear}
    boundBone {boneName=dyn_front_skirt_3}
    componentLinearAccScale=<1 1 1>
    componentLinearVelScale=<1 1 1>
    gravityOverride {z=-980}
    physicsBodyDefinitions [{
      boundBone {boneName=dyn_front_skirt_3}
      boxExtents=<30 30 30>
      constraintSetup {angularLimitsMax=<0 0 0>, angularLimitsMin=<0 -10 -15>, angularTarget=<1 0 0>}
      localJointOffset=<4 0 0>
     }]
   }
   n15 = Anim Dynamics - Bone: dyn_R_side_skirt_1!(ComponentPose=@n14, Alpha=1, GravityOverride:Vector="0,0,-980", ExternalForce="0,0,0") -> (Pose) {
    angularDampingOverride=1
    angularSpringConstant=5000
    bAngularSpring=true
    bOverrideAngularDamping=true
    bUseGravityOverride=true
    gravityScale=0.35
    simulationSpace=RootRelative
    alphaBoolBlend {blendOption=Linear}
    boundBone {boneName=dyn_R_side_skirt_1}
    componentLinearAccScale=<1 1 1>
    componentLinearVelScale=<1 1 1>
    gravityOverride {z=-980}
    physicsBodyDefinitions [{
      boundBone {boneName=dyn_R_side_skirt_1}
      boxExtents=<25 25 25>
      constraintSetup {angularLimitsMax=<0 0 0>, angularLimitsMin=<0 0 -30>, angularTarget=<1 0 0>}
      localJointOffset=<4 0 0>
     }]
   }
   n16 = Anim Dynamics - Bone: dyn_R_side_skirt_2!(ComponentPose=@n15, Alpha=1, GravityOverride:Vector="0,0,-980", ExternalForce="0,0,0") -> (Pose) {
    angularDampingOverride=1
    angularSpringConstant=5000
    bAngularSpring=true
    bOverrideAngularDamping=true
    bUseGravityOverride=true
    gravityScale=0.35
    simulationSpace=RootRelative
    alphaBoolBlend {blendOption=Linear}
    boundBone {boneName=dyn_R_side_skirt_2}
    componentLinearAccScale=<1 1 1>
    componentLinearVelScale=<1 1 1>
    gravityOverride {z=-980}
    physicsBodyDefinitions [{
      boundBone {boneName=dyn_R_side_skirt_2}
      boxExtents=<25 25 25>
      constraintSetup {angularLimitsMax=<0 0 0>, angularLimitsMin=<0 0 -30>, angularTarget=<1 0 0>}
      localJointOffset=<4 0 0>
     }]
   }
   n17 = Anim Dynamics - Bone: dyn_R_side_skirt_3!(ComponentPose=@n16, Alpha=1, GravityOverride:Vector="0,0,-980", ExternalForce="0,0,0") -> (Pose) {
    angularDampingOverride=1
    angularSpringConstant=5000
    bAngularSpring=true
    bOverrideAngularDamping=true
    bUseGravityOverride=true
    gravityScale=0.35
    simulationSpace=RootRelative
    alphaBoolBlend {blendOption=Linear}
    boundBone {boneName=dyn_R_side_skirt_3}
    componentLinearAccScale=<1 1 1>
    componentLinearVelScale=<1 1 1>
    gravityOverride {z=-980}
    physicsBodyDefinitions [{
      boundBone {boneName=dyn_R_side_skirt_3}
      boxExtents=<25 25 25>
      constraintSetup {angularLimitsMax=<0 0 0>, angularLimitsMin=<0 0 -30>, angularTarget=<1 0 0>}
      localJointOffset=<4 0 0>
     }]
   }
   n18 = Anim Dynamics - Bone: dyn_R_back_skirt_1!(ComponentPose=@n17, Alpha=1, GravityOverride:Vector="0,0,-980", ExternalForce="0,0,0") -> (Pose) {
    angularDampingOverride=1
    angularSpringConstant=5000
    bAngularSpring=true
    bOverrideAngularDamping=true
    bUseGravityOverride=true
    gravityScale=0.35
    simulationSpace=RootRelative
    alphaBoolBlend {blendOption=Linear}
    boundBone {boneName=dyn_R_back_skirt_1}
    componentLinearAccScale=<1 1 1>
    componentLinearVelScale=<1 1 1>
    gravityOverride {z=-980}
    physicsBodyDefinitions [{
      boundBone {boneName=dyn_R_back_skirt_1}
      boxExtents=<25 25 25>
      constraintSetup {angularLimitsMax=<0 10 0>, angularLimitsMin=<0 0 -30>, angularTarget=<1 0 0>}
      localJointOffset=<4 0 0>
     }]
   }
   n19 = Anim Dynamics - Bone: dyn_R_back_skirt_2!(ComponentPose=@n18, Alpha=1, GravityOverride:Vector="0,0,-980", ExternalForce="0,0,0") -> (Pose) {
    angularDampingOverride=1
    angularSpringConstant=5000
    bAngularSpring=true
    bOverrideAngularDamping=true
    bUseGravityOverride=true
    gravityScale=0.35
    simulationSpace=RootRelative
    alphaBoolBlend {blendOption=Linear}
    boundBone {boneName=dyn_R_back_skirt_2}
    componentLinearAccScale=<1 1 1>
    componentLinearVelScale=<1 1 1>
    gravityOverride {z=-980}
    physicsBodyDefinitions [{
      boundBone {boneName=dyn_R_back_skirt_2}
      boxExtents=<25 25 25>
      constraintSetup {angularLimitsMax=<0 10 0>, angularLimitsMin=<0 0 -30>, angularTarget=<1 0 0>}
      localJointOffset=<4 0 0>
     }]
   }
   n20 = Anim Dynamics - Bone: dyn_R_back_skirt_3!(ComponentPose=@n19, Alpha=1, GravityOverride:Vector="0,0,-980", ExternalForce="0,0,0") -> (Pose) {
    angularDampingOverride=1
    angularSpringConstant=5000
    bAngularSpring=true
    bOverrideAngularDamping=true
    bUseGravityOverride=true
    gravityScale=0.35
    simulationSpace=RootRelative
    alphaBoolBlend {blendOption=Linear}
    boundBone {boneName=dyn_R_back_skirt_3}
    componentLinearAccScale=<1 1 1>
    componentLinearVelScale=<1 1 1>
    gravityOverride {z=-980}
    physicsBodyDefinitions [{
      boundBone {boneName=dyn_R_back_skirt_3}
      boxExtents=<25 25 25>
      constraintSetup {angularLimitsMax=<0 10 0>, angularLimitsMin=<0 0 -30>, angularTarget=<1 0 0>}
      localJointOffset=<4 0 0>
     }]
   }
   n21 = "Transform (Modify) Bone - Bone: dyn_L_skirtDriver_1"(ComponentPose=@n20, Alpha=1, Translation="0,-2,0", Rotation="0,0,0", Scale="1,1,1") {
    rotationMode=BMM_Additive
    rotationSpace=BCS_BoneSpace
    translationMode=BMM_Additive
    translationSpace=BCS_BoneSpace
    alphaBoolBlend {blendOption=Linear}
    boneToModify {boneName=dyn_L_skirtDriver_1}
    translation {y=-2}
   }
   Output Pose(Result=Component To Local(ComponentPose=@n21)) {layerGroup=None}
  ]}
 EventGraph {nodes [
   Event BlueprintUpdateAnimation()
   > n1 = Gravity Override State Function DEPRECATED(CharacterPartAnimInstance=Self-Reference(), GameTimeInSeconds=GetGameTimeInSeconds(), GravityOverrideParams=Get Gravity Override Params())
   > Set GravityOverride(GravityOverride=@n1)
   > Branch(Condition=Get bIsInFrontEnd())
   > Set isPlayingFrontendIdle(isPlayingFrontendIdle:float=GetCurveValue(self=GetOwnerFrontEndAnimInstance(), CurveName=Frontend_Idle_Bypass)) -> (Output_Get:float)
  ]}
]
variables [
 Gravity Override Params {type=GravityOverrideParamsStruct}
 GravityOverride {type=Vector}
 isPlayingFrontendIdle {type=number}
]
