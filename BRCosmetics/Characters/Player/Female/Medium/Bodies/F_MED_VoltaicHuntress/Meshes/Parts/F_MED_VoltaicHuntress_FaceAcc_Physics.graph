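# Physics asset for the Voltaic Huntress face accessory mesh, defining capsule bodies and constraints for dynamic hair on the female medium character.
# Two five-segment main hair chains and two branching side-hair chains hang from the head with limited swing and twist, the side limits differing left to right.
# Extra angular-free constraints cross-link the lower three left and right main segments.
# Spine, neck, head, clavicle and upper-arm capsules form the body, joined by angular-free constraints.
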PhysicsAsset /BRCosmetics/Characters/Player/Female/Medium/Bodies/F_MED_VoltaicHuntress/Meshes/Parts/F_MED_VoltaicHuntress_FaceAcc_Physics
bodies {
 clavicle_l [{shapeType=Capsule, length=7, radius=8, shapeName=clavicle_r_capsule, center=<4.0438 0.8606 -3.546>, rotation=<rot 89.6647 92.4944 87.5056>}]
 clavicle_r [{shapeType=Capsule, length=7, radius=8, shapeName=clavicle_r_capsule, center=<-4.0438 -0.8606 3.546>, rotation=<rot 90 0 0>}]
 dyn_L_hair_main_1 [{shapeType=Capsule, length=7, radius=4, shapeName=dyn_L_hair_main_1_capsule, center=<-4 0 0>, rotation=<rot 88.1217 -87.7389 -72.8563>}]
 dyn_L_hair_main_2 [{shapeType=Capsule, length=10, radius=5, shapeName=dyn_L_hair_main_2_capsule, center=<-4.4873 2.95 0>, rotation=<rot 90 0 0>}]
 dyn_L_hair_main_3 [{shapeType=Capsule, length=10, radius=5, shapeName=dyn_L_hair_main_3_capsule, center=<-4.2445 3.6822 0>, rotation=<rot 90 0 0>}]
 dyn_L_hair_main_4 [{shapeType=Capsule, length=10, radius=3, shapeName=dyn_L_hair_main_4_capsule, center=<-3.4649 1.2005 0>, rotation=<rot 90 0 0>}]
 dyn_L_hair_main_5 [{shapeType=Capsule, length=7, radius=5, shapeName=dyn_L_hair_main_5_capsule, center=<-3 2 0>, rotation=<rot 90 0 0>}]
 dyn_L_sideHair_1 [{shapeType=Capsule, length=3, radius=2, shapeName=dyn_L_sideHair_1_capsule, center=<2 0 0>, rotation=<rot 90 0 0>}]
 dyn_L_sideHair_2 [{shapeType=Capsule, length=3, radius=1.5, shapeName=dyn_L_sideHair_2_capsule, center=<2 0 0>, rotation=<rot 90 0 0>}]
 dyn_L_sideHair_strand_A [{shapeType=Capsule, length=6, radius=1.6, shapeName=dyn_L_sideHair_strand_A_capsule, center=<3 0 0>, rotation=<rot 90 0 0>}]
 dyn_L_sideHair_strand_B [{shapeType=Capsule, length=6, radius=1.6, shapeName=dyn_L_sideHair_strand_B_capsule, center=<3 0 0>, rotation=<rot 90 0 0>}]
 dyn_R_hair_main_1 [{shapeType=Capsule, length=7, radius=4, shapeName=dyn_R_hair_main_1_capsule, center=<4 0 0>, rotation=<rot 88.1217 -87.7389 -72.8563>}]
 dyn_R_hair_main_2 [{
   shapeType=Capsule
   length=10
   radius=5
   shapeName=dyn_R_hair_main_2_capsule
   center=<5.0712 -2.7255 -2.4698>
   rotation=<rot 89.5337 -69.8185 -59.8293>
  }]
 dyn_R_hair_main_3 [{shapeType=Capsule, length=10, radius=5, shapeName=dyn_R_hair_main_3_capsule, center=<4.2445 -3.6822 0>, rotation=<rot 90 0 0>}]
 dyn_R_hair_main_4 [{shapeType=Capsule, length=10, radius=3, shapeName=dyn_R_hair_main_4_capsule, center=<3.4649 -1.2005 0>, rotation=<rot 90 0 0>}]
 dyn_R_hair_main_5 [{shapeType=Capsule, length=7, radius=5, shapeName=dyn_R_hair_main_5_capsule, center=<3 -2 0>, rotation=<rot 90 0 0>}]
 dyn_R_sideHair_1 [{shapeType=Capsule, length=3, radius=2, shapeName=dyn_R_sideHair_1_capsule, center=<2 0 0>, rotation=<rot 90 0 0>}]
 dyn_R_sideHair_2 [{shapeType=Capsule, length=3, radius=0.5, shapeName=dyn_R_sideHair_2_capsule, center=<2 0 0>, rotation=<rot 90 0 0>}]
 dyn_R_sideHair_strand_A [{shapeType=Capsule, length=5, radius=1.2, shapeName=dyn_R_sideHair_strand_A_capsule, center=<3 0 0>, rotation=<rot 90 0 0>}]
 dyn_R_sideHair_strand_B [{shapeType=Capsule, length=6, radius=1.6, shapeName=dyn_R_sideHair_strand_B_capsule, center=<3 0 0>, rotation=<rot 90 0 0>}]
 head [{shapeType=Capsule, length=6, radius=5, shapeName=head_capsule, center=<-0.0213 2.582 0>, rotation=<rot 90 0 0>}]
 neck_01 [{shapeType=Capsule, length=60, radius=20, shapeName=neck_01_capsule, center=<3.0304 18.9243 0>}]
 spine_03 [{shapeType=Capsule, length=10, radius=8, shapeName=spine_03_capsule, center=<0 1.032 0>}]
 spine_04 [{shapeType=Capsule, length=10, radius=10, shapeName=spine_04_capsule, center=<0 1.9595 0>}]
 spine_05 [{shapeType=Capsule, length=10, radius=9, shapeName=spine_05_capsule, center=<-0.0373 0.7077 0>}]
 upperarm_l [{shapeType=Capsule, length=23.4241, radius=6, shapeName=upperarm_l_capsule, center=<11 0 0>, rotation=<rot 90 0 0>}]
 upperarm_r [{shapeType=Capsule, length=23.4241, radius=6, shapeName=upperarm_l_capsule, center=<-11 0 0>, rotation=<rot 90 0 0>}]
}
constraints [
 {
  bone1Name=spine_04
  bone2Name=spine_03
  swing1LimitDegrees=45
  swing1Motion=Free
  swing2LimitDegrees=45
  swing2Motion=Free
  twistLimitDegrees=45
  twistMotion=Free
 }
 {
  bone1Name=spine_05
  bone2Name=spine_04
  swing1LimitDegrees=45
  swing1Motion=Free
  swing2LimitDegrees=45
  swing2Motion=Free
  twistLimitDegrees=45
  twistMotion=Free
 }
 {
  bone1Name=clavicle_l
  bone2Name=spine_05
  swing1LimitDegrees=45
  swing1Motion=Free
  swing2LimitDegrees=45
  swing2Motion=Free
  twistLimitDegrees=45
  twistMotion=Free
 }
 {
  bone1Name=upperarm_l
  bone2Name=clavicle_l
  swing1LimitDegrees=45
  swing1Motion=Free
  swing2LimitDegrees=45
  swing2Motion=Free
  twistLimitDegrees=45
  twistMotion=Free
 }
 {
  bone1Name=clavicle_r
  bone2Name=spine_05
  swing1LimitDegrees=45
  swing1Motion=Free
  swing2LimitDegrees=45
  swing2Motion=Free
  twistLimitDegrees=45
  twistMotion=Free
 }
 {
  bone1Name=upperarm_r
  bone2Name=clavicle_r
  swing1LimitDegrees=45
  swing1Motion=Free
  swing2LimitDegrees=45
  swing2Motion=Free
  twistLimitDegrees=45
  twistMotion=Free
 }
 {
  bone1Name=head
  bone2Name=neck_01
  swing1LimitDegrees=45
  swing1Motion=Free
  swing2LimitDegrees=45
  swing2Motion=Free
  twistLimitDegrees=45
  twistMotion=Free
 }
 {bone1Name=dyn_L_hair_main_1, bone2Name=head, swing1LimitDegrees=50, swing2LimitDegrees=7, twistLimitDegrees=10}
 {
  bone1Name=neck_01
  bone2Name=spine_05
  swing1LimitDegrees=45
  swing1Motion=Free
  swing2LimitDegrees=45
  swing2Motion=Free
  twistLimitDegrees=45
  twistMotion=Free
 }
 {bone1Name=dyn_L_hair_main_2, bone2Name=dyn_L_hair_main_1, swing1LimitDegrees=50, swing2LimitDegrees=7, twistLimitDegrees=10}
 {bone1Name=dyn_L_hair_main_3, bone2Name=dyn_L_hair_main_2, swing1LimitDegrees=45, swing2LimitDegrees=7, twistLimitDegrees=10}
 {bone1Name=dyn_L_hair_main_4, bone2Name=dyn_L_hair_main_3, swing1LimitDegrees=45, swing2LimitDegrees=7, twistLimitDegrees=10}
 {bone1Name=dyn_L_hair_main_5, bone2Name=dyn_L_hair_main_4, swing1LimitDegrees=45, swing2LimitDegrees=7, twistLimitDegrees=10}
 {bone1Name=dyn_R_hair_main_1, bone2Name=head, swing1LimitDegrees=50, swing2LimitDegrees=7, twistLimitDegrees=10}
 {bone1Name=dyn_R_hair_main_2, bone2Name=dyn_R_hair_main_1, swing1LimitDegrees=50, swing2LimitDegrees=7, twistLimitDegrees=10}
 {bone1Name=dyn_R_hair_main_3, bone2Name=dyn_R_hair_main_2, swing1LimitDegrees=45, swing2LimitDegrees=7, twistLimitDegrees=10}
 {bone1Name=dyn_R_hair_main_4, bone2Name=dyn_R_hair_main_3, swing1LimitDegrees=45, swing2LimitDegrees=7, twistLimitDegrees=10}
 {bone1Name=dyn_R_hair_main_5, bone2Name=dyn_R_hair_main_4, swing1LimitDegrees=45, swing2LimitDegrees=7, twistLimitDegrees=10}
 {
  bone1Name=dyn_L_hair_main_3
  bone2Name=dyn_R_hair_main_3
  swing1LimitDegrees=45
  swing1Motion=Free
  swing2LimitDegrees=45
  swing2Motion=Free
  twistLimitDegrees=45
  twistMotion=Free
 }
 {
  bone1Name=dyn_L_hair_main_4
  bone2Name=dyn_R_hair_main_4
  swing1LimitDegrees=45
  swing1Motion=Free
  swing2LimitDegrees=45
  swing2Motion=Free
  twistLimitDegrees=45
  twistMotion=Free
 }
 {
  bone1Name=dyn_L_hair_main_5
  bone2Name=dyn_R_hair_main_5
  swing1LimitDegrees=45
  swing1Motion=Free
  swing2LimitDegrees=45
  swing2Motion=Free
  twistLimitDegrees=45
  twistMotion=Free
 }
 {bone1Name=dyn_L_sideHair_1, bone2Name=head, swing1LimitDegrees=20, swing2LimitDegrees=30, twistLimitDegrees=10}
 {bone1Name=dyn_L_sideHair_2, bone2Name=dyn_L_sideHair_1, swing1LimitDegrees=20, swing2LimitDegrees=45, twistLimitDegrees=10}
 {bone1Name=dyn_L_sideHair_strand_A, bone2Name=dyn_L_sideHair_2, swing1LimitDegrees=20, swing2LimitDegrees=20, twistLimitDegrees=10}
 {bone1Name=dyn_L_sideHair_strand_B, bone2Name=dyn_L_sideHair_2, swing1LimitDegrees=10, swing2LimitDegrees=20, twistLimitDegrees=10}
 {bone1Name=dyn_R_sideHair_1, bone2Name=head, swing1LimitDegrees=20, swing2LimitDegrees=5, twistLimitDegrees=10}
 {bone1Name=dyn_R_sideHair_2, bone2Name=dyn_R_sideHair_1, swing1LimitDegrees=20, swing2LimitDegrees=5, twistLimitDegrees=10}
 {bone1Name=dyn_R_sideHair_strand_A, bone2Name=dyn_R_sideHair_2, swing1LimitDegrees=20, swing2LimitDegrees=5, twistLimitDegrees=10}
 {bone1Name=dyn_R_sideHair_strand_B, bone2Name=dyn_R_sideHair_2, swing1LimitDegrees=20, swing2LimitDegrees=5, twistLimitDegrees=10}
]
