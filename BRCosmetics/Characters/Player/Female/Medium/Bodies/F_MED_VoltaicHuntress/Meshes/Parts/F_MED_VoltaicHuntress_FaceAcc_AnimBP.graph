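# Post-process AnimBP for the Voltaic Huntress face accessory that adds secondary motion to her hair.
# It copies the parent mesh pose, then chains RigidBody, weighted bone constraints, per-bone AnimDynamics and additive offsets, spreading head, body and main-hair motion across strands.
# In the front end, the Frontend_Idle_Bypass curve fades out right side-hair AnimDynamics.
# The event graph also computes state- and backbling-based GravityOverride, but the AnimGraph uses fixed gravity values.
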AnimBlueprint /BRCosmetics/Characters/Player/Female/Medium/Bodies/F_MED_VoltaicHuntress/Meshes/Parts/F_MED_VoltaicHuntress_FaceAcc_AnimBP
target_skeleton=&/F_MED_VoltaicHuntress_FaceAcc_Skeleton
graphs [
 AnimGraph {nodes [
   n1 = RigidBody!(
    ComponentPose=Local To Component(LocalPose=Copy Pose From Mesh() {bUseAttachedParent=true})
    Alpha
    OverrideWorldGravity:Vector="0,0,-980"
    ExternalForce
   ) -> (Pose) {
    bClampLinearTranslationLimitToRefPose=true
    bOverrideWorldGravity=true
    overridePhysicsAsset=&/F_MED_VoltaicHuntress_FaceAcc_Physics
    simulationSpace=BaseBoneSpace
    alphaBoolBlend {blendOption=Linear}
    baseBoneRef {boneName=spine_05}
    componentAppliedLinearAccClamp=<5000 5000 5000>
    componentLinearAccScale=<0.3 0.3 0.3>
    componentLinearVelScale=<0.4 0.4 0.4>
    overrideWorldGravity {z=-980}
    simSpaceSettings {
     maxAngularAcceleration=1000
     maxAngularVelocity=1000
     maxLinearAcceleration=1000
     maxLinearVelocity=1000
     velocityScaleZ=0.05
     worldAlpha=0.5
     externalLinearDragV=<1 1 1>
    }
   }
   n2 = "Constraint - dyn_L_sideHair_long_2 (PPP)"!(ComponentPose=@n1, Alpha=1, ConstraintWeights_0=0.2, ConstraintWeights_1:float=0.7, ConstraintWeights_2:float=0.1) -> (Pose) {
    alphaBoolBlend {blendOption=Linear}
    boneToModify {boneName=dyn_L_sideHair_long_2}
    constraintSetup [{targetBone {boneName=head}}, {targetBone {boneName=dyn_L_hair_main_1}}, {targetBone {boneName=dyn_L_hair_main_2}}]
    constraintWeights [1,1,1]
   }
   n3 = "Constraint - dyn_L_sideHair_long_3 (PPPP)"!(
    ComponentPose=@n2
    Alpha=1
    ConstraintWeights_0=0.1
    ConstraintWeights_1:float=0.5
    ConstraintWeights_2:float=0.3
    ConstraintWeights_3:float=0.1
   ) -> (Pose) {
    alphaBoolBlend {blendOption=Linear}
    boneToModify {boneName=dyn_L_sideHair_long_3}
    constraintSetup [
     {targetBone {boneName=head}}
     {targetBone {boneName=dyn_L_hair_main_1}}
     {targetBone {boneName=dyn_L_hair_main_2}}
     {targetBone {boneName=dyn_L_hair_main_3}}
    ]
    constraintWeights [1,1,1,1]
   }
   n4 = "Constraint - dyn_L_sideHair_long_4 (PPP)"!(ComponentPose=@n3, Alpha=1, ConstraintWeights_0=0.4, ConstraintWeights_1:float=0.5, ConstraintWeights_2:float=0.1) -> (Pose) {
    alphaBoolBlend {blendOption=Linear}
    boneToModify {boneName=dyn_L_sideHair_long_4}
    constraintSetup [{targetBone {boneName=dyn_L_hair_main_3}}, {targetBone {boneName=dyn_L_hair_main_2}}, {targetBone {boneName=dyn_L_hair_main_1}}]
    constraintWeights [1,1,1]
   }
   n5 = "Constraint - dyn_L_sideHair_long_5 (PPPP)"!(
    ComponentPose=@n4
    Alpha=1
    ConstraintWeights_0=0.5
    ConstraintWeights_1:float=0.2
    ConstraintWeights_2:float=0.2
    ConstraintWeights_3:float=0.1
   ) -> (Pose) {
    alphaBoolBlend {blendOption=Linear}
    boneToModify {boneName=dyn_L_sideHair_long_5}
    constraintSetup [
     {targetBone {boneName=dyn_L_hair_main_3}}
     {targetBone {boneName=dyn_L_hair_main_2}}
     {targetBone {boneName=dyn_L_hair_main_4}}
     {targetBone {boneName=dyn_R_hair_main_2}}
    ]
    constraintWeights [1,1,1,1]
   }
   n6 = "Constraint - dyn_R_sideHair_long_3 (PPPPPPPP)"!(
    ComponentPose=@n5
    Alpha=1
    ConstraintWeights_0=0.3646
    ConstraintWeights_1:float=0.3315
    ConstraintWeights_2:float=0.1466
    ConstraintWeights_3:float=0.1326
    ConstraintWeights_4:float=0.0092
    ConstraintWeights_5:float=0.007
    ConstraintWeights_6:float=0.0048
    ConstraintWeights_7:float=0.0036
   ) -> (Pose) {
    alphaBoolBlend {blendOption=Linear}
    boneToModify {boneName=dyn_R_sideHair_long_3}
    constraintSetup [
     {targetBone {boneName=dyn_R_hair_main_2}}
     {targetBone {boneName=dyn_R_hair_main_1}}
     {targetBone {boneName=dyn_R_hair_main_3}}
     {targetBone {boneName=head}}
     {targetBone {boneName=dyn_L_hair_main_1}}
     {targetBone {boneName=dyn_L_hair_main_2}}
     {targetBone {boneName=dyn_R_hair_main_4}}
     {targetBone {boneName=dyn_L_hair_main_3}}
    ]
    constraintWeights [1,1,1,1,1,1,1,1]
   }
   n7 = "Constraint - dyn_R_sideHair_long_A (PPPPPPPP)"!(
    ComponentPose=@n6
    Alpha=1
    ConstraintWeights_0=0.4371
    ConstraintWeights_1:float=0.2863
    ConstraintWeights_2:float=0.1989
    ConstraintWeights_3:float=0.0631
    ConstraintWeights_4:float=0.0053
    ConstraintWeights_5:float=0.0031
    ConstraintWeights_6:float=0.0031
    ConstraintWeights_7:float=0.003
   ) -> (Pose) {
    alphaBoolBlend {blendOption=Linear}
    boneToModify {boneName=dyn_R_sideHair_long_A}
    constraintSetup [
     {targetBone {boneName=dyn_R_hair_main_2}}
     {targetBone {boneName=dyn_R_hair_main_1}}
     {targetBone {boneName=dyn_R_hair_main_3}}
     {targetBone {boneName=head}}
     {targetBone {boneName=dyn_R_hair_main_4}}
     {targetBone {boneName=dyn_L_hair_main_1}}
     {targetBone {boneName=dyn_L_hair_main_3}}
     {targetBone {boneName=dyn_L_hair_main_2}}
    ]
    constraintWeights [1,1,1,1,1,1,1,1]
   }
   n8 = "Constraint - dyn_R_sideHair_long_4 (PPPP)"!(
    ComponentPose=@n7
    Alpha=1
    ConstraintWeights_0=0.1
    ConstraintWeights_1:float=0
    ConstraintWeights_2:float=0.2
    ConstraintWeights_3:float=0.7
   ) -> (Pose) {
    alphaBoolBlend {blendOption=Linear}
    boneToModify {boneName=dyn_R_sideHair_long_4}
    constraintSetup [
     {targetBone {boneName=dyn_R_hair_main_2}}
     {targetBone {boneName=dyn_R_hair_main_3}}
     {targetBone {boneName=dyn_R_hair_main_1}}
     {targetBone {boneName=clavicle_r}}
    ]
    constraintWeights [1,1,1,1]
   }
   n9 = "Constraint - dyn_R_sideHair_long_5 (PPPPPPPP)"!(
    ComponentPose=@n8
    Alpha=1
    ConstraintWeights_0=0.0012
    ConstraintWeights_1:float=0.0112
    ConstraintWeights_2:float=0.0043
    ConstraintWeights_3:float=0.0251
    ConstraintWeights_4:float=0.1959
    ConstraintWeights_5:float=0.5991
    ConstraintWeights_6:float=0.1546
    ConstraintWeights_7:float=0.0086
   ) -> (Pose) {
    alphaBoolBlend {blendOption=Linear}
    boneToModify {boneName=dyn_R_sideHair_long_5}
    constraintSetup [
     {targetBone {boneName=dyn_L_hair_main_2}}
     {targetBone {boneName=dyn_L_hair_main_3}}
     {targetBone {boneName=dyn_L_hair_main_4}}
     {targetBone {boneName=dyn_R_hair_main_1}}
     {targetBone {boneName=dyn_R_hair_main_2}}
     {targetBone {boneName=dyn_R_hair_main_3}}
     {targetBone {boneName=dyn_R_hair_main_4}}
     {targetBone {boneName=dyn_R_hair_main_5}}
    ]
    constraintWeights [1,1,1,1,1,1,1,1]
   }
   n10 = "Constraint - dyn_R_sideHair_long_6 (PPPPPPPP)"!(
    ComponentPose=@n9
    Alpha=1
    ConstraintWeights_0=0.537
    ConstraintWeights_1:float=0.3214
    ConstraintWeights_2:float=0.0932
    ConstraintWeights_3:float=0.0256
    ConstraintWeights_4:float=0.0095
    ConstraintWeights_5:float=0.0082
    ConstraintWeights_6:float=0.0043
    ConstraintWeights_7:float=0.0007
   ) -> (Pose) {
    alphaBoolBlend {blendOption=Linear}
    boneToModify {boneName=dyn_R_sideHair_long_6}
    constraintSetup [
     {targetBone {boneName=dyn_R_hair_main_4}}
     {targetBone {boneName=dyn_R_hair_main_3}}
     {targetBone {boneName=dyn_R_hair_main_5}}
     {targetBone {boneName=dyn_R_hair_main_2}}
     {targetBone {boneName=dyn_L_hair_main_4}}
     {targetBone {boneName=dyn_L_hair_main_3}}
     {targetBone {boneName=dyn_L_hair_main_5}}
     {targetBone {boneName=dyn_R_hair_main_1}}
    ]
    constraintWeights [1,1,1,1,1,1,1,1]
   }
   n11 = "Constraint - dyn_R_sideHair_long_7 (PPPPPPP)"!(
    ComponentPose=@n10
    Alpha=1
    ConstraintWeights_0=0.5201
    ConstraintWeights_1:float=0.3453
    ConstraintWeights_2:float=0.088
    ConstraintWeights_3:float=0.0258
    ConstraintWeights_4:float=0.0096
    ConstraintWeights_5:float=0.0093
    ConstraintWeights_6:float=0.0019
   ) -> (Pose) {
    alphaBoolBlend {blendOption=Linear}
    boneToModify {boneName=dyn_R_sideHair_long_7}
    constraintSetup [
     {targetBone {boneName=dyn_R_hair_main_4}}
     {targetBone {boneName=dyn_R_hair_main_5}}
     {targetBone {boneName=dyn_R_hair_main_3}}
     {targetBone {boneName=dyn_L_hair_main_5}}
     {targetBone {boneName=dyn_L_hair_main_4}}
     {targetBone {boneName=dyn_R_hair_main_2}}
     {targetBone {boneName=dyn_L_hair_main_3}}
    ]
    constraintWeights [1,1,1,1,1,1,1]
   }
   n12 = Anim Dynamics - Bone: dyn_front_strand!(ComponentPose=@n11, Alpha=1, GravityOverride:Vector="0,0,-980", ExternalForce="0,0,0") -> (Pose) {
    angularDampingOverride=1
    angularSpringConstant=2000
    bAngularSpring=true
    bOverrideAngularDamping=true
    bUseGravityOverride=true
    gravityScale=0.35
    simulationSpace=RootRelative
    alphaBoolBlend {blendOption=Linear}
    boundBone {boneName=dyn_front_strand}
    componentLinearAccScale=<1 1 1>
    componentLinearVelScale=<1 1 1>
    gravityOverride {z=-980}
    physicsBodyDefinitions [{
      boundBone {boneName=dyn_front_strand}
      boxExtents=<12 12 12>
      constraintSetup {angularLimitsMax=<15 1 30>, angularLimitsMin=<-15 -20 0>, angularTarget=<1 0 0>}
      localJointOffset=<2 0 0>
     }]
   }
   n13 = Anim Dynamics - Bone: dyn_L_side_strand_1!(ComponentPose=@n12, Alpha=1, GravityOverride:Vector="0,0,-980", ExternalForce="0,0,0") -> (Pose) {
    angularDampingOverride=1
    angularSpringConstant=2000
    bAngularSpring=true
    bOverrideAngularDamping=true
    bUseGravityOverride=true
    gravityScale=0.35
    simulationSpace=RootRelative
    alphaBoolBlend {blendOption=Linear}
    boundBone {boneName=dyn_L_side_strand_1}
    componentLinearAccScale=<1 1 1>
    componentLinearVelScale=<1 1 1>
    gravityOverride {z=-980}
    physicsBodyDefinitions [{
      boundBone {boneName=dyn_L_side_strand_1}
      boxExtents=<18 18 18>
      constraintSetup {angularLimitsMax=<3 3 0>, angularLimitsMin=<-3 -3 -15>, angularTarget=<1 0 0>}
      localJointOffset=<4 0 0>
     }]
   }
   n14 = Anim Dynamics - Bone: dyn_L_side_strand_2!(ComponentPose=@n13, Alpha=1, GravityOverride:Vector="0,0,-980", ExternalForce="0,0,0") -> (Pose) {
    angularDampingOverride=1
    angularSpringConstant=2000
    bAngularSpring=true
    bOverrideAngularDamping=true
    bUseGravityOverride=true
    gravityScale=0.35
    simulationSpace=RootRelative
    alphaBoolBlend {blendOption=Linear}
    boundBone {boneName=dyn_L_side_strand_2}
    componentLinearAccScale=<1 1 1>
    componentLinearVelScale=<1 1 1>
    gravityOverride {z=-980}
    physicsBodyDefinitions [{
      boundBone {boneName=dyn_L_side_strand_2}
      boxExtents=<18 18 18>
      constraintSetup {angularLimitsMax=<3 3 0>, angularLimitsMin=<-3 -3 -15>, angularTarget=<1 0 0>}
      localJointOffset=<4 0 0>
     }]
   }
   n15 = Anim Dynamics - Bone: dyn_L_sideHair_long_2!(ComponentPose=@n14, Alpha=1, GravityOverride:Vector="0,0,-980", ExternalForce="0,0,0") -> (Pose) {
    angularDampingOverride=1
    angularSpringConstant=200
    bAngularSpring=true
    bOverrideAngularDamping=true
    bUseGravityOverride=true
    gravityScale=0.35
    simulationSpace=RootRelative
    alphaBoolBlend {blendOption=Linear}
    boundBone {boneName=dyn_L_sideHair_long_2}
    componentLinearAccScale=<1 1.1 1.1>
    componentLinearVelScale=<1 1.1 1.1>
    gravityOverride {z=-980}
    physicsBodyDefinitions [{
      boundBone {boneName=dyn_L_sideHair_long_2}
      boxExtents=<58 58 58>
      constraintSetup {angularLimitsMax=<0 0 0>, angularLimitsMin=<0 -5 -7>, angularTarget=<1 0 0>}
      localJointOffset=<4 0 0>
     }]
   }
   n16 = Anim Dynamics - Bone: dyn_L_sideHair_long_3!(ComponentPose=@n15, Alpha=1, GravityOverride:Vector="0,0,-980", ExternalForce="0,0,0") -> (Pose) {
    angularDampingOverride=1
    angularSpringConstant=200
    bAngularSpring=true
    bOverrideAngularDamping=true
    bUseGravityOverride=true
    gravityScale=0.35
    simulationSpace=RootRelative
    alphaBoolBlend {blendOption=Linear}
    boundBone {boneName=dyn_L_sideHair_long_3}
    componentLinearAccScale=<1 1.1 1.1>
    componentLinearVelScale=<1 1.1 1.1>
    gravityOverride {z=-980}
    physicsBodyDefinitions [{
      boundBone {boneName=dyn_L_sideHair_long_3}
      boxExtents=<58 58 58>
      constraintSetup {angularLimitsMax=<0 0 0>, angularLimitsMin=<0 -5 -7>, angularTarget=<1 0 0>}
      localJointOffset=<4 0 0>
     }]
   }
   n17 = Anim Dynamics - Bone: dyn_L_sideHair_long_4!(ComponentPose=@n16, Alpha=1, GravityOverride:Vector="0,0,-980", ExternalForce="0,0,0") -> (Pose) {
    angularDampingOverride=1
    angularSpringConstant=200
    bAngularSpring=true
    bOverrideAngularDamping=true
    bUseGravityOverride=true
    gravityScale=0.35
    simulationSpace=RootRelative
    alphaBoolBlend {blendOption=Linear}
    boundBone {boneName=dyn_L_sideHair_long_4}
    componentLinearAccScale=<1 1.1 1.1>
    componentLinearVelScale=<1 1.1 1.1>
    gravityOverride {z=-980}
    physicsBodyDefinitions [{
      boundBone {boneName=dyn_L_sideHair_long_4}
      boxExtents=<58 58 58>
      constraintSetup {angularLimitsMax=<0 0 0>, angularLimitsMin=<0 -5 -7>, angularTarget=<1 0 0>}
      localJointOffset=<4 0 0>
     }]
   }
   n18 = Anim Dynamics - Bone: dyn_L_sideHair_long_5!(ComponentPose=@n17, Alpha=1, GravityOverride:Vector="0,0,-980", ExternalForce="0,0,0") -> (Pose) {
    angularDampingOverride=1
    angularSpringConstant=200
    bAngularSpring=true
    bOverrideAngularDamping=true
    bUseGravityOverride=true
    gravityScale=0.35
    simulationSpace=RootRelative
    alphaBoolBlend {blendOption=Linear}
    boundBone {boneName=dyn_L_sideHair_long_5}
    componentLinearAccScale=<1 1.1 1.1>
    componentLinearVelScale=<1 1.1 1.1>
    gravityOverride {z=-980}
    physicsBodyDefinitions [{
      boundBone {boneName=dyn_L_sideHair_long_5}
      boxExtents=<58 58 58>
      constraintSetup {angularLimitsMax=<0 0 0>, angularLimitsMin=<0 -5 -7>, angularTarget=<1 0 0>}
      localJointOffset=<4 0 0>
     }]
   }
   n19 = Anim Dynamics - Bone: dyn_L_sideHair_long_6!(ComponentPose=@n18, Alpha=1, GravityOverride:Vector="0,0,-980", ExternalForce="0,0,0") -> (Pose) {
    angularDampingOverride=1
    angularSpringConstant=200
    bAngularSpring=true
    bOverrideAngularDamping=true
    bUseGravityOverride=true
    gravityScale=0.35
    simulationSpace=RootRelative
    alphaBoolBlend {blendOption=Linear}
    boundBone {boneName=dyn_L_sideHair_long_6}
    componentLinearAccScale=<1 1.1 1.1>
    componentLinearVelScale=<1 1.1 1.1>
    gravityOverride {z=-980}
    physicsBodyDefinitions [{
      boundBone {boneName=dyn_L_sideHair_long_6}
      boxExtents=<58 58 58>
      constraintSetup {angularLimitsMax=<0 0 0>, angularLimitsMin=<0 -5 -7>, angularTarget=<1 0 0>}
      localJointOffset=<4 0 0>
     }]
   }
   n20 = Get turnOffDuringCustomIdle() -> (turnOffDuringCustomIdle:double)
   n21 = Anim Dynamics - Bone: dyn_R_sideHair_long_2!(ComponentPose=@n19, Alpha=@n20, GravityOverride:Vector="0,0,-980", ExternalForce="0,0,0") -> (Pose) {
    angularDampingOverride=1
    angularSpringConstant=200
    bAngularSpring=true
    bOverrideAngularDamping=true
    bUseGravityOverride=true
    gravityScale=0.35
    simulationSpace=RootRelative
    alphaBoolBlend {blendOption=Linear}
    boundBone {boneName=dyn_R_sideHair_long_2}
    componentLinearAccScale=<1 1.1 1.1>
    componentLinearVelScale=<1 1.1 1.1>
    gravityOverride {z=-980}
    physicsBodyDefinitions [{
      boundBone {boneName=dyn_R_sideHair_long_2}
      boxExtents=<40 40 40>
      constraintSetup {angularLimitsMax=<0 0 0>, angularLimitsMin=<0 0 -5>, angularTarget=<1 0 0>}
      localJointOffset=<4 0 0>
     }]
   }
   n22 = Anim Dynamics - Bone: dyn_R_sideHair_long_3!(ComponentPose=@n21, Alpha=@n20, GravityOverride:Vector="0,0,-980", ExternalForce="0,0,0") -> (Pose) {
    angularDampingOverride=1
    angularSpringConstant=200
    bAngularSpring=true
    bOverrideAngularDamping=true
    bUseGravityOverride=true
    gravityScale=0.35
    simulationSpace=RootRelative
    alphaBoolBlend {blendOption=Linear}
    boundBone {boneName=dyn_R_sideHair_long_3}
    componentLinearAccScale=<1 1.1 1.1>
    componentLinearVelScale=<1 1.1 1.1>
    gravityOverride {z=-980}
    physicsBodyDefinitions [{
      boundBone {boneName=dyn_R_sideHair_long_3}
      boxExtents=<40 40 40>
      constraintSetup {angularLimitsMax=<0 5 0>, angularLimitsMin=<0 0 -5>, angularTarget=<1 0 0>}
      localJointOffset=<4 0 0>
     }]
   }
   n23 = Anim Dynamics - Bone: dyn_R_sideHair_long_A!(ComponentPose=@n22, Alpha=@n20, GravityOverride:Vector="0,0,-980", ExternalForce="0,0,0") -> (Pose) {
    angularDampingOverride=1
    angularSpringConstant=200
    bAngularSpring=true
    bOverrideAngularDamping=true
    bUseGravityOverride=true
    gravityScale=0.35
    simulationSpace=RootRelative
    alphaBoolBlend {blendOption=Linear}
    boundBone {boneName=dyn_R_sideHair_long_A}
    componentLinearAccScale=<1 1.1 1.1>
    componentLinearVelScale=<1 1.1 1.1>
    gravityOverride {z=-980}
    physicsBodyDefinitions [{
      boundBone {boneName=dyn_R_sideHair_long_A}
      boxExtents=<40 40 40>
      constraintSetup {angularLimitsMax=<0 3 0>, angularLimitsMin=<0 0 -15>, angularTarget=<1 0 0>}
      localJointOffset=<4 0 0>
     }]
   }
   n24 = Anim Dynamics - Bone: dyn_R_sideHair_long_4!(ComponentPose=@n23, Alpha=@n20, GravityOverride:Vector="0,0,-980", ExternalForce="0,0,0") -> (Pose) {
    angularDampingOverride=1
    angularSpringConstant=200
    bAngularSpring=true
    bOverrideAngularDamping=true
    bUseGravityOverride=true
    gravityScale=0.35
    simulationSpace=RootRelative
    alphaBoolBlend {blendOption=Linear}
    boundBone {boneName=dyn_R_sideHair_long_4}
    componentLinearAccScale=<1 1.1 1.1>
    componentLinearVelScale=<1 1.1 1.1>
    gravityOverride {z=-980}
    physicsBodyDefinitions [{
      boundBone {boneName=dyn_R_sideHair_long_4}
      boxExtents=<40 40 40>
      constraintSetup {angularLimitsMax=<0 5 0>, angularLimitsMin=<0 0 -7>, angularTarget=<1 0 0>}
      localJointOffset=<4 0 0>
     }]
   }
   n25 = Anim Dynamics - Bone: dyn_R_sideHair_long_5!(ComponentPose=@n24, Alpha=@n20, GravityOverride:Vector="0,0,-980", ExternalForce="0,0,0") -> (Pose) {
    angularDampingOverride=1
    angularSpringConstant=200
    bAngularSpring=true
    bOverrideAngularDamping=true
    bUseGravityOverride=true
    gravityScale=0.35
    simulationSpace=RootRelative
    alphaBoolBlend {blendOption=Linear}
    boundBone {boneName=dyn_R_sideHair_long_5}
    componentLinearAccScale=<1 1.1 1.1>
    componentLinearVelScale=<1 1.1 1.1>
    gravityOverride {z=-980}
    physicsBodyDefinitions [{
      boundBone {boneName=dyn_R_sideHair_long_5}
      boxExtents=<40 40 40>
      constraintSetup {angularLimitsMax=<0 5 0>, angularLimitsMin=<0 0 -7>, angularTarget=<1 0 0>}
      localJointOffset=<4 0 0>
     }]
   }
   n26 = Anim Dynamics - Bone: dyn_R_sideHair_long_6!(ComponentPose=@n25, Alpha=@n20, GravityOverride:Vector="0,0,-980", ExternalForce="0,0,0") -> (Pose) {
    angularDampingOverride=1
    angularSpringConstant=200
    bAngularSpring=true
    bOverrideAngularDamping=true
    bUseGravityOverride=true
    gravityScale=0.35
    simulationSpace=RootRelative
    alphaBoolBlend {blendOption=Linear}
    boundBone {boneName=dyn_R_sideHair_long_6}
    componentLinearAccScale=<1 1.1 1.1>
    componentLinearVelScale=<1 1.1 1.1>
    gravityOverride {z=-980}
    physicsBodyDefinitions [{
      boundBone {boneName=dyn_R_sideHair_long_6}
      boxExtents=<40 40 40>
      constraintSetup {angularLimitsMax=<0 5 0>, angularLimitsMin=<0 0 -7>, angularTarget=<1 0 0>}
      localJointOffset=<4 0 0>
     }]
   }
   n27 = Anim Dynamics - Bone: dyn_R_sideHair_long_7!(ComponentPose=@n26, Alpha=@n20, GravityOverride:Vector="0,0,-980", ExternalForce="0,0,0") -> (Pose) {
    angularDampingOverride=1
    angularSpringConstant=200
    bAngularSpring=true
    bOverrideAngularDamping=true
    bUseGravityOverride=true
    gravityScale=0.35
    simulationSpace=RootRelative
    alphaBoolBlend {blendOption=Linear}
    boundBone {boneName=dyn_R_sideHair_long_7}
    componentLinearAccScale=<1 1.1 1.1>
    componentLinearVelScale=<1 1.1 1.1>
    gravityOverride {z=-980}
    physicsBodyDefinitions [{
      boundBone {boneName=dyn_R_sideHair_long_7}
      boxExtents=<40 40 40>
      constraintSetup {angularLimitsMax=<0 5 0>, angularLimitsMin=<0 0 -7>, angularTarget=<1 0 0>}
      localJointOffset=<4 0 0>
     }]
   }
   n28 = "Constraint - dyn_L_hair_main_1 (P)"(ComponentPose=@n27, Alpha=1, ConstraintWeights_0=1) {
    alphaBoolBlend {blendOption=Linear}
    boneToModify {boneName=dyn_L_hair_main_1}
    constraintSetup [{targetBone {boneName=neck_02}}]
    constraintWeights [1]
   }
   n29 = "Constraint - dyn_R_hair_main_1 (P)"(ComponentPose=@n28, Alpha=1, ConstraintWeights_0=1) {
    alphaBoolBlend {blendOption=Linear}
    boneToModify {boneName=dyn_R_hair_main_1}
    constraintSetup [{targetBone {boneName=neck_02}}]
    constraintWeights [1]
   }
   n30 = "Transform (Modify) Bone - Bone: dyn_L_hair_main_1"!(ComponentPose=@n29, Alpha=1, Translation="4,0,0", Scale="1,1,1") -> (Pose) {
    rotationMode=BMM_Additive
    rotationSpace=BCS_BoneSpace
    translationMode=BMM_Additive
    alphaBoolBlend {blendOption=Linear}
    boneToModify {boneName=dyn_L_hair_main_1}
    translation {x=4}
   }
   n31 = "Transform (Modify) Bone - Bone: dyn_R_hair_main_1"!(ComponentPose=@n30, Alpha=1, Translation="4,0,0", Scale="1,1,1") -> (Pose) {
    rotationMode=BMM_Additive
    rotationSpace=BCS_BoneSpace
    translationMode=BMM_Additive
    alphaBoolBlend {blendOption=Linear}
    boneToModify {boneName=dyn_R_hair_main_1}
    translation {x=4}
   }
   n32 = "Constraint - dyn_L_hair_main_3 (PPPPP)"!(
    ComponentPose=@n31
    Alpha=1
    ConstraintWeights_0=0.5961
    ConstraintWeights_1:float=0.3114
    ConstraintWeights_2:float=0.0472
    ConstraintWeights_3:float=0.0269
    ConstraintWeights_4:float=0.0184
   ) -> (Pose) {
    alphaBoolBlend {blendOption=Linear}
    alphaScaleBiasClamp {scale=0.8}
    boneToModify {boneName=dyn_L_hair_main_3}
    constraintSetup [
     {targetBone {boneName=clavicle_l}}
     {targetBone {boneName=spine_05}}
     {targetBone {boneName=spine_04}}
     {targetBone {boneName=upperarm_twist_01_l}}
     {targetBone {boneName=clavicle_r}}
    ]
    constraintWeights [1,1,1,1,1]
   }
   n33 = "Transform (Modify) Bone - Bone: dyn_L_hair_main_3"!(ComponentPose=@n32, Alpha=1, Translation="-2,3,0", Scale="1,1,1") -> (Pose) {
    rotationMode=BMM_Additive
    rotationSpace=BCS_BoneSpace
    translationMode=BMM_Additive
    alphaBoolBlend {blendOption=Linear}
    boneToModify {boneName=dyn_L_hair_main_3}
    translation=<-2 3>
   }
   n34 = "Constraint - dyn_L_hair_main_4 (PPPPPPP)"!(
    ComponentPose=@n33
    Alpha=1
    ConstraintWeights_0=0.4711
    ConstraintWeights_1:float=0.204
    ConstraintWeights_2:float=0.1966
    ConstraintWeights_3:float=0.0997
    ConstraintWeights_4:float=0.0256
    ConstraintWeights_5:float=0.0019
    ConstraintWeights_6:float=0.001
   ) -> (Pose) {
    alphaBoolBlend {blendOption=Linear}
    alphaScaleBiasClamp {scale=0.8}
    boneToModify {boneName=dyn_L_hair_main_4}
    constraintSetup [
     {targetBone {boneName=spine_04}}
     {targetBone {boneName=spine_05}}
     {targetBone {boneName=clavicle_l}}
     {targetBone {boneName=upperarm_twist_01_l}}
     {targetBone {boneName=clavicle_r}}
     {targetBone {boneName=upperarm_twist_01_r}}
     {targetBone {boneName=spine_02}}
    ]
    constraintWeights [1,1,1,1,1,1,1]
   }
   n35 = "Constraint - dyn_L_hair_main_5 (PPPPPPPP)"!(
    ComponentPose=@n34
    Alpha=1
    ConstraintWeights_0=0.5133
    ConstraintWeights_1:float=0.174
    ConstraintWeights_2:float=0.0952
    ConstraintWeights_3:float=0.0611
    ConstraintWeights_4:float=0.0608
    ConstraintWeights_5:float=0.0467
    ConstraintWeights_6:float=0.0285
    ConstraintWeights_7:float=0.0204
   ) -> (Pose) {
    alphaBoolBlend {blendOption=Linear}
    boneToModify {boneName=dyn_L_hair_main_5}
    constraintSetup [
     {targetBone {boneName=spine_04}}
     {targetBone {boneName=spine_02}}
     {targetBone {boneName=spine_05}}
     {targetBone {boneName=upperarm_twist_01_l}}
     {targetBone {boneName=clavicle_l}}
     {targetBone {boneName=spine_03}}
     {targetBone {boneName=spine_01}}
     {targetBone {boneName=clavicle_r}}
    ]
    constraintWeights [1,1,1,1,1,1,1,1]
   }
   n36 = "Constraint - dyn_R_hair_main_3 (PPPPPP)"!(
    ComponentPose=@n35
    Alpha=1
    ConstraintWeights_0=0.4726
    ConstraintWeights_1:float=0.3902
    ConstraintWeights_2:float=0.0682
    ConstraintWeights_3:float=0.0377
    ConstraintWeights_4:float=0.0313
    ConstraintWeights_5:float=0.0001
   ) -> (Pose) {
    alphaBoolBlend {blendOption=Linear}
    alphaScaleBiasClamp {scale=0.8}
    boneToModify {boneName=dyn_R_hair_main_3}
    constraintSetup [
     {targetBone {boneName=clavicle_r}}
     {targetBone {boneName=spine_05}}
     {targetBone {boneName=spine_04}}
     {targetBone {boneName=upperarm_twist_01_r}}
     {targetBone {boneName=clavicle_l}}
     {targetBone {boneName=upperarm_twist_01_l}}
    ]
    constraintWeights [1,1,1,1,1,1]
   }
   n37 = "Transform (Modify) Bone - Bone: dyn_R_hair_main_3"!(ComponentPose=@n36, Alpha=1, Translation="-2,3,0", Scale="1,1,1") -> (Pose) {
    rotationMode=BMM_Additive
    rotationSpace=BCS_BoneSpace
    translationMode=BMM_Additive
    alphaBoolBlend {blendOption=Linear}
    boneToModify {boneName=dyn_R_hair_main_3}
    translation=<-2 3>
   }
   n38 = "Constraint - dyn_R_hair_main_4 (PPPPP)"!(
    ComponentPose=@n37
    Alpha=1
    ConstraintWeights_0=0.4915
    ConstraintWeights_1:float=0.2174
    ConstraintWeights_2:float=0.1573
    ConstraintWeights_3:float=0.1027
    ConstraintWeights_4:float=0.0312
   ) -> (Pose) {
    alphaBoolBlend {blendOption=Linear}
    alphaScaleBiasClamp {scale=0.8}
    boneToModify {boneName=dyn_R_hair_main_4}
    constraintSetup [
     {targetBone {boneName=spine_04}}
     {targetBone {boneName=spine_05}}
     {targetBone {boneName=clavicle_r}}
     {targetBone {boneName=upperarm_twist_01_r}}
     {targetBone {boneName=clavicle_l}}
    ]
    constraintWeights [1,1,1,1,1]
   }
   n39 = "Constraint - dyn_R_hair_main_5 (PPPPPPPP)"!(
    ComponentPose=@n38
    Alpha=1
    ConstraintWeights_0=0.516
    ConstraintWeights_1:float=0.1706
    ConstraintWeights_2:float=0.0959
    ConstraintWeights_3:float=0.0616
    ConstraintWeights_4:float=0.0606
    ConstraintWeights_5:float=0.0452
    ConstraintWeights_6:float=0.0301
    ConstraintWeights_7:float=0.02
   ) -> (Pose) {
    alphaBoolBlend {blendOption=Linear}
    boneToModify {boneName=dyn_R_hair_main_5}
    constraintSetup [
     {targetBone {boneName=spine_04}}
     {targetBone {boneName=spine_02}}
     {targetBone {boneName=spine_05}}
     {targetBone {boneName=clavicle_r}}
     {targetBone {boneName=upperarm_twist_01_r}}
     {targetBone {boneName=spine_03}}
     {targetBone {boneName=spine_01}}
     {targetBone {boneName=clavicle_l}}
    ]
    constraintWeights [1,1,1,1,1,1,1,1]
   }
   n40 = "Transform (Modify) Bone - Bone: dyn_L_sideHair_long_1"!(ComponentPose=@n39, Alpha=1, Translation="0,0,0", Scale="1,1,1") -> (Pose) {
    rotationMode=BMM_Additive
    rotationSpace=BCS_BoneSpace
    alphaBoolBlend {blendOption=Linear}
    boneToModify {boneName=dyn_L_sideHair_long_1}
    rotation {pitch=10, yaw=-15}
   }
   n41 = "Transform (Modify) Bone - Bone: dyn_L_sideHair_long_3"!(ComponentPose=@n40, Alpha=1, Translation="0,0,0", Scale="1,1,1") -> (Pose) {
    rotationMode=BMM_Additive
    rotationSpace=BCS_BoneSpace
    alphaBoolBlend {blendOption=Linear}
    boneToModify {boneName=dyn_L_sideHair_long_3}
    rotation {yaw=25}
   }
   n42 = "Transform (Modify) Bone - Bone: dyn_R_sideHair_long_1"!(ComponentPose=@n41, Alpha=1, Translation="0,0,0", Scale="1,1,1") -> (Pose) {
    rotationMode=BMM_Additive
    rotationSpace=BCS_BoneSpace
    alphaBoolBlend {blendOption=Linear}
    boneToModify {boneName=dyn_R_sideHair_long_1}
    rotation {pitch=-10, yaw=-20}
   }
   n43 = "Transform (Modify) Bone - Bone: dyn_R_sideHair_long_3"!(ComponentPose=@n42, Alpha=1, Translation="0,0,0", Scale="1,1,1") -> (Pose) {
    rotationMode=BMM_Additive
    rotationSpace=BCS_BoneSpace
    alphaBoolBlend {blendOption=Linear}
    boneToModify {boneName=dyn_R_sideHair_long_3}
    rotation {yaw=35}
   }
   n44 = "Constraint - dyn_L_hair_main_3 (PPPPP)"(
    ComponentPose=@n43
    Alpha=1
    ConstraintWeights_0=0.5961
    ConstraintWeights_1=0.3114
    ConstraintWeights_2=0.0472
    ConstraintWeights_3=0.0269
    ConstraintWeights_4=0.0184
   ) {
    alphaBoolBlend {blendOption=Linear}
    boneToModify {boneName=dyn_L_hair_main_3}
    constraintSetup [
     {targetBone {boneName=clavicle_l}}
     {targetBone {boneName=spine_05}}
     {targetBone {boneName=spine_04}}
     {targetBone {boneName=upperarm_twist_01_l}}
     {targetBone {boneName=clavicle_r}}
    ]
    constraintWeights [1,1,1,1,1]
   }
   n45 = "Constraint - dyn_L_hair_main_4 (PPPPPPP)"(
    ComponentPose=@n44
    Alpha=1
    ConstraintWeights_0=0.4711
    ConstraintWeights_1=0.204
    ConstraintWeights_2=0.1966
    ConstraintWeights_3=0.0997
    ConstraintWeights_4=0.0256
    ConstraintWeights_5=0.0019
    ConstraintWeights_6=0.001
   ) {
    alphaBoolBlend {blendOption=Linear}
    boneToModify {boneName=dyn_L_hair_main_4}
    constraintSetup [
     {targetBone {boneName=spine_04}}
     {targetBone {boneName=spine_05}}
     {targetBone {boneName=clavicle_l}}
     {targetBone {boneName=upperarm_twist_01_l}}
     {targetBone {boneName=clavicle_r}}
     {targetBone {boneName=upperarm_twist_01_r}}
     {targetBone {boneName=spine_02}}
    ]
    constraintWeights [1,1,1,1,1,1,1]
   }
   n46 = "Constraint - dyn_R_hair_main_3 (PPPPPP)"(
    ComponentPose=@n45
    Alpha=1
    ConstraintWeights_0=0.4726
    ConstraintWeights_1=0.3902
    ConstraintWeights_2=0.0682
    ConstraintWeights_3=0.0377
    ConstraintWeights_4=0.0313
    ConstraintWeights_5=0.0001
   ) {
    alphaBoolBlend {blendOption=Linear}
    boneToModify {boneName=dyn_R_hair_main_3}
    constraintSetup [
     {targetBone {boneName=clavicle_r}}
     {targetBone {boneName=spine_05}}
     {targetBone {boneName=spine_04}}
     {targetBone {boneName=upperarm_twist_01_r}}
     {targetBone {boneName=clavicle_l}}
     {targetBone {boneName=upperarm_twist_01_l}}
    ]
    constraintWeights [1,1,1,1,1,1]
   }
   n47 = "Constraint - dyn_R_hair_main_4 (PPPPP)"(
    ComponentPose=@n46
    Alpha=1
    ConstraintWeights_0=0.4915
    ConstraintWeights_1=0.2174
    ConstraintWeights_2=0.1573
    ConstraintWeights_3=0.1027
    ConstraintWeights_4=0.0312
   ) {
    alphaBoolBlend {blendOption=Linear}
    boneToModify {boneName=dyn_R_hair_main_4}
    constraintSetup [
     {targetBone {boneName=spine_04}}
     {targetBone {boneName=spine_05}}
     {targetBone {boneName=clavicle_r}}
     {targetBone {boneName=upperarm_twist_01_r}}
     {targetBone {boneName=clavicle_l}}
    ]
    constraintWeights [1,1,1,1,1]
   }
   n48 = "Constraint - dyn_R_sideHair_long_A (PPPPPPPP)"(
    ComponentPose=@n47
    Alpha=1
    ConstraintWeights_0=0.0631
    ConstraintWeights_1=0.0031
    ConstraintWeights_2=0.003
    ConstraintWeights_3=0.0031
    ConstraintWeights_4=0.2863
    ConstraintWeights_5=0.4371
    ConstraintWeights_6=0.1989
    ConstraintWeights_7=0.0053
   ) {
    alphaBoolBlend {blendOption=Linear}
    boneToModify {boneName=dyn_R_sideHair_long_A}
    constraintSetup [
     {targetBone {boneName=head}}
     {targetBone {boneName=dyn_L_hair_main_1}}
     {targetBone {boneName=dyn_L_hair_main_2}}
     {targetBone {boneName=dyn_L_hair_main_3}}
     {targetBone {boneName=dyn_R_hair_main_1}}
     {targetBone {boneName=dyn_R_hair_main_2}}
     {targetBone {boneName=dyn_R_hair_main_3}}
     {targetBone {boneName=dyn_R_hair_main_4}}
    ]
    constraintWeights [1,1,1,1,1,1,1,1]
   }
   n49 = "Constraint - dyn_R_sideHair_long_5 (PPPPPPPP)"(
    ComponentPose=@n48
    Alpha=1
    ConstraintWeights_0=0.0012
    ConstraintWeights_1=0.0112
    ConstraintWeights_2=0.0043
    ConstraintWeights_3=0.0251
    ConstraintWeights_4=0.1959
    ConstraintWeights_5=0.5991
    ConstraintWeights_6=0.1546
    ConstraintWeights_7=0.0086
   ) {
    alphaBoolBlend {blendOption=Linear}
    boneToModify {boneName=dyn_R_sideHair_long_5}
    constraintSetup [
     {targetBone {boneName=dyn_L_hair_main_2}}
     {targetBone {boneName=dyn_L_hair_main_3}}
     {targetBone {boneName=dyn_L_hair_main_4}}
     {targetBone {boneName=dyn_R_hair_main_1}}
     {targetBone {boneName=dyn_R_hair_main_2}}
     {targetBone {boneName=dyn_R_hair_main_3}}
     {targetBone {boneName=dyn_R_hair_main_4}}
     {targetBone {boneName=dyn_R_hair_main_5}}
    ]
    constraintWeights [1,1,1,1,1,1,1,1]
   }
   n50 = "Constraint - dyn_R_sideHair_2 (PPPPPPPP)"!(
    ComponentPose=@n49
    Alpha=1
    ConstraintWeights_0=0.0141
    ConstraintWeights_1:float=0.104
    ConstraintWeights_2:float=0.1167
    ConstraintWeights_3:float=0.2275
    ConstraintWeights_4:float=0.1348
    ConstraintWeights_5:float=0.1625
    ConstraintWeights_6:float=0.1241
    ConstraintWeights_7:float=0.1163
   ) -> (Pose) {
    alphaBoolBlend {blendOption=Linear}
    boneToModify {boneName=dyn_R_sideHair_2}
    constraintSetup [
     {targetBone {boneName=dyn_R_sideHair_long_3}}
     {targetBone {boneName=dyn_R_hair_main_2}}
     {targetBone {boneName=dyn_R_hair_main_1}}
     {targetBone {boneName=dyn_R_sideHair_long_1}}
     {targetBone {boneName=dyn_R_sideHair_1}}
     {targetBone {boneName=head}}
     {targetBone {boneName=clavicle_r}}
     {targetBone {boneName=dyn_R_sideHair_2}}
    ]
    constraintWeights [1,1,1,1,1,1,1,1]
   }
   Output Pose(Result=Component To Local(ComponentPose=@n50)) {layerGroup=None}
  ]}
 EventGraph {nodes [
   Event OnCharacterCustomizationLoaded()
   > Set isWearingBackblingBool(isWearingBackblingBool:bool=CheckForBackbling()) -> (Output_Get:bool)
   Event BlueprintUpdateAnimation()
   > n1 = Gravity Override State Function DEPRECATED(CharacterPartAnimInstance=Self-Reference(), GameTimeInSeconds=GetGameTimeInSeconds(), GravityOverrideParams=Get Gravity Override Params())
   > n2 = Set GravityOverride(GravityOverride=@n1)
   > Set GravityOverride(GravityOverride=SelectVector(A="0,0,-5000", B=@n2, bPickA=Get isWearingBackblingBool() -> (isWearingBackblingBool:bool)))
   > Branch(Condition=Get bIsInFrontEnd())
   n3 = GetCurveValue(self=GetOwnerFrontEndAnimInstance(), CurveName=Frontend_Idle_Bypass)
   > Set isPlayingFrontendIdle(isPlayingFrontendIdle:float=@n3) -> (Output_Get:float)
   > Set turnOffDuringCustomIdle(turnOffDuringCustomIdle:double=float - float(A=1, B=@n3)) -> (Output_Get:double)
  ]}
]
variables [
 Gravity Override Params {
  type=GravityOverrideParamsStruct
  default {
   onGround_Falling {gravityOverride=<0 0 -900>}
   onGround_Standing {gravityOverride=<0 0 -750>}
   skydive_FloatingDown {gravityOverride=<0 -700 100>, windAmplitude=<100 -250 200>}
  }
 }
 GravityOverride {type=Vector}
 isPlayingFrontendIdle {type=number}
 isWearingBackblingBool {type=boolean}
 turnOffDuringCustomIdle {type=number}
]
